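# Control Rig module that rocks a base on the rim of a radius-sized disc toward an Aim To target.
# Construction builds Plane, Base and End nulls from the Driven Joint, plus an Aim To control or null.
# Forward solve projects the aim onto the plane to find a rim hinge, using the law of sines to scale the aim vector.
# Helper functions re-pivot and aim the base, apply driver offsets, and draw debug shapes.
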ControlRigBlueprint /EpicControlRig/Modules/Miscellaneous/CRM_AimRocker_v01
graphs [
 ConstructionGraph {nodes [
   RigUnit_PrepareForExecution = &RigUnit_PrepareForExecution()
   RigVMFunction_Sequence_1 = RigVMAggregateNode(ExecuteContext=@RigUnit_PrepareForExecution.ExecutePin)
   n1 = &RigUnit_HierarchyAddNull(
    Name=Plane
    Space=GlobalSpace
    Parent=/Script/ControlRig.RigUnit_ResolveConnector(Connector=(Type=Connector,Name="Parent"), SkipSocket=False) {is=struct RigVMUnitNode}.Result
    ExecutePin=@RigVMFunction_Sequence_1.B
    Transform=RigVMVariableNode() {variable=Plane Transform}.Value
   )
   > RigVMVariableNode(Value=@n1.Item) {variable=Plane Element}
   n2 = &RigVMFunction_ControlFlowBranch(Condition=RigVMVariableNode() {variable=Create Control}.Value, ExecuteContext=@RigVMFunction_Sequence_1.C)
   n3 = RigVMVariableNode() {variable=Aim To Transform}
   n4 = &RigUnit_ResolveConnector(Connector=(Type=Connector,Name="Aim To Parent"), SkipSocket=False)
   n5 = /Script/ControlRig.RigUnit_HierarchyAddControlTransform(
    InitialValue=(Rotation=(X=0,Y=0,Z=0,W=1),Translation=(X=0,Y=0,Z=0),Scale3D=(X=1,Y=1,Z=1))
    Name=Aim To
    OffsetSpace=GlobalSpace
    Settings="(InitialSpace=LocalSpace,bUsePreferredRotationOrder=False,PreferredRotationOrder=YZX,Limits=(LimitTranslationX=(bMinimum=False,bMaximum=False),LimitTranslationY=(bMinimum=False,bMaximum=False),LimitTranslationZ=(bMinimum=False,bMaximum=False),LimitPitch=(b ...(truncated)"
    ExecutePin=@n2.True
    OffsetTransform=@n3.Value
    Settings.Shape=RigVMVariableNode() {variable=Control Shape}.Value
    Parent=@n4.Result
   ) {
    is=struct RigVMUnitNode
    template="SpawnControl::Execute(in InitialValue,in Settings,in OffsetTransform,in OffsetSpace,in Parent,in Name,out Item)"
   }
   > RigVMVariableNode(Value=@n5.Item) {variable=Aim To Element}
   n6 = &RigUnit_HierarchyAddNull(Name=Aim To, Space=GlobalSpace, Transform=@n3.Value, Parent=@n4.Result, ExecutePin=@n2.False)
   > RigVMVariableNode(Value=@n6.Item) {variable=Aim To Element}
   Power = /Script/RigVM.RigVMFunction_MathDoublePow(B=2, A=RigVMVariableNode() {variable=Radius}.Value) {is=struct RigVMUnitNode, template="Power::Execute(in A,in B,out Result)"}
   Add = &RigVMFunction_MathDoubleAdd(
    A=@Power.Result
    B=&RigVMFunction_MathDoublePow(B=2, A=RigVMVariableNode() {variable=Length}.Value) {template="Power::Execute(in A,in B,out Result)"}.Result
   )
   Absolute = &RigVMFunction_MathDoubleAbs(Value=&RigVMFunction_MathDoublePow(B=0.5, A=@Add.Result) {template="Power::Execute(in A,in B,out Result)"}.Result)
   RigVMVariableNode(Value=@Absolute.Result, ExecuteContext=@RigVMFunction_Sequence_1.E) {variable=Hypotenuse}
   n7 = &RigUnit_HierarchyAddNull(
    Name=Base
    Space=GlobalSpace
    Parent=RigVMVariableNode() {variable=Plane Element}.Value
    Transform=RigVMVariableNode() {variable=Plane Transform}.Value
    ExecutePin=@RigVMFunction_Sequence_1.D
   )
   > RigVMVariableNode(Value=@n7.Item) {variable=Base Element}
   n8 = RigVMVariableNode() {variable=Plane Transform}
   RigVMFunction_MathVectorSetLength = /Script/RigVM.RigVMFunction_MathVectorSetLength(
    Value=/Script/RigVM.RigVMFunction_MathTransformRotateVector(Transform=@n8.Value, Vector=RigVMVariableNode() {variable=Primary Axis}.Value) {is=struct RigVMUnitNode, template="Rotate Vector::Execute(in Transform,in Vector,out Result)"}.Result
    Length=RigVMVariableNode() {variable=Length}.Value
   ) {is=struct RigVMUnitNode, template="SetLength::Execute(in Value,in Length,out Result)"}
   > n9 = &RigUnit_HierarchyAddNull(
    Name=End
    Space=GlobalSpace
    Transform=(Rotation=(X=0,Y=0,Z=0,W=1),Translation=(X=0,Y=0,Z=0),Scale3D=(X=1,Y=1,Z=1))
    Parent=RigVMVariableNode() {variable=Base Element}.Value
    Transform.Scale3D=@n8.Value.Scale3D
    Transform.Rotation=@n8.Value.Rotation
    Transform.Translation=/Script/RigVM.RigVMFunction_MathVectorAdd(B=@RigVMFunction_MathVectorSetLength.Result, A=@n8.Value.Translation) {is=struct RigVMUnitNode, template="Add::Execute(in A,in B,out Result)"}.Result
   )
   > RigVMVariableNode(Value=@n9.Item) {variable=End Element}
   n10 = &RigUnit_GetTransform(bInitial=False, Item=&RigUnit_ResolveConnector(Connector=(Type=Connector,Name="Driven Joint"), SkipSocket=False).Result)
   RigVMVariableNode(Value=@n10.Transform, ExecuteContext=@RigVMFunction_Sequence_1.A) {variable=Plane Transform}
   n11 = &RigVMFunction_MathDoubleAdd(A=RigVMVariableNode() {variable=Length}.Value, B=RigVMVariableNode() {variable=Aim Extra Length}.Value)
   n12 = &RigVMFunction_MathTransformRotateVector(Transform=@n10.Transform, Vector=RigVMVariableNode() {variable=Primary Axis}.Value) {template="Rotate Vector::Execute(in Transform,in Vector,out Result)"}
   > RigVMVariableNode(
    Value=(Rotation=(X=0,Y=0,Z=0,W=1),Translation=(X=0,Y=0,Z=0),Scale3D=(X=1,Y=1,Z=1))
    Value.Translation=&RigVMFunction_MathVectorAdd(
     A=&RigVMFunction_MathVectorSetLength(Length=@n11.Result, Value=@n12.Result) {template="SetLength::Execute(in Value,in Length,out Result)"}.Result
     B=@n10.Transform.Translation
    ) {template="Add::Execute(in A,in B,out Result)"}.Result
    Value.Scale3D=@n10.Transform.Scale3D
    Value.Rotation=@n10.Transform.Rotation
   ) {variable=Aim To Transform}
  ]}
 New Function_ContainedGraph {nodes [
   RigVMFunctionEntryNode()
   > RigVMFunction_Sequence = RigVMAggregateNode()
   RigVMVariableNode(
    Value=&RigUnit_GetTransform(bInitial=False, Item=RigVMVariableNode() {variable=Plane Element}.Value).Transform
    ExecuteContext=@RigVMFunction_Sequence.A
   ) {variable=Return_Plane Transform}
   > RigVMVariableNode(Value=&RigUnit_GetTransform(bInitial=False, Item=RigVMVariableNode() {variable=Aim To Element}.Value).Transform) {variable=Return_Aim To Transform}
   n1 = RigVMVariableNode() {variable=Return_Plane Transform}
   n2 = /Script/RigVM.RigVMFunction_MathVectorSub(A=RigVMVariableNode() {variable=Return_Aim To Transform}.Value.Translation, B=@n1.Value.Translation) {is=struct RigVMUnitNode, template="Subtract::Execute(in A,in B,out Result)"}
   RigVMVariableNode(Value=@n2.Result, ExecuteContext=@RigVMFunction_Sequence.B) {variable=Return_Aim Vector}
   > RigVMVariableNode(Value=&RigVMFunction_MathTransformRotateVector(Vector=RigVMVariableNode() {variable=Primary Axis}.Value, Transform=@n1.Value) {template="Rotate Vector::Execute(in Transform,in Vector,out Result)"}.Result) {variable=Return_Plane Normal}
   > RigVMVariableNode(Value=&RigVMFunction_MathVectorAdd(A=@n2.Result, B=@n1.Value.Translation) {template="Add::Execute(in A,in B,out Result)"}.Result) {variable=Return_Aim Tip}
   Return = RigVMFunctionReturnNode(
    Plane Normal=RigVMVariableNode() {variable=Return_Plane Normal}.Value
    Aim Vector=RigVMVariableNode() {variable=Return_Aim Vector}.Value
    Aim Tip=RigVMVariableNode() {variable=Return_Aim Tip}.Value
    Proj Vector=RigVMVariableNode() {variable=Return_Proj Vector}.Value
    Proj Point=RigVMVariableNode() {variable=Return_Proj Point}.Value
    ExecuteContext=@RigVMFunction_Sequence.E
   )
   Get Scaled Aim Vector = RigVMFunctionReferenceNode(
    Aim Vector=RigVMVariableNode() {variable=Return_Aim Vector}.Value
    Hypothenuse=RigVMVariableNode() {variable=Hypothenuse}.Value
    Radius=RigVMVariableNode() {variable=Radius}.Value
    Proj Vector=RigVMVariableNode() {variable=Return_Proj Vector}.Value
    Length=RigVMVariableNode() {variable=Length}.Value
    ExecuteContext=@RigVMFunction_Sequence.D
   )
   > RigVMVariableNode(Value=@"Get Scaled Aim Vector".Scaled Aim Vector) {variable=Return_Aim Vector}
   Add = &RigVMFunction_MathVectorAdd(B=RigVMVariableNode() {variable=Return_Plane Transform}.Value.Translation, A=@"Get Scaled Aim Vector".Scaled Aim Vector) {template="Add::Execute(in A,in B,out Result)"}
   > RigVMVariableNode(Value=@Add.Result) {variable=Return_Aim Tip}
   Subtract = &RigVMFunction_MathVectorSub(
    A=/Script/RigVM.RigVMFunction_MathDistanceToPlane(
     PlaneNormal=RigVMVariableNode() {variable=Return_Plane Normal}.Value
     Point=RigVMVariableNode() {variable=Return_Aim Tip}.Value
     PlanePoint=RigVMVariableNode() {variable=Return_Plane Transform}.Value.Translation
    ) {is=struct RigVMUnitNode}.ClosestPointOnPlane
    B=RigVMVariableNode() {variable=Return_Plane Transform}.Value.Translation
   ) {template="Subtract::Execute(in A,in B,out Result)"}
   n3 = RigVMVariableNode(Value=@Subtract.Result, ExecuteContext=@RigVMFunction_Sequence.C) {variable=Return_Proj Vector}
   RigVMFunction_MathVectorLength = /Script/RigVM.RigVMFunction_MathVectorLength(Value=@Subtract.Result) {is=struct RigVMUnitNode, template="Length::Execute(in Value,out Result)"}
   Greater = /Script/RigVM.RigVMFunction_MathDoubleGreater(B=0, A=@RigVMFunction_MathVectorLength.Result) {is=struct RigVMUnitNode, template="Greater::Execute(in A,in B,out Result)"}
   RigVMFunction_ControlFlowBranch = &RigVMFunction_ControlFlowBranch(Condition=@Greater.Result, ExecuteContext=@n3.ExecuteContext)
   RigVMFunction_MathVectorSetLength = &RigVMFunction_MathVectorSetLength(Value=RigVMVariableNode() {variable=Return_Proj Vector}.Value, Length=RigVMVariableNode() {variable=Radius}.Value) {template="SetLength::Execute(in Value,in Length,out Result)"}
   RigVMVariableNode(ExecuteContext=@RigVMFunction_ControlFlowBranch.True, Value=@RigVMFunction_MathVectorSetLength.Result) {variable=Return_Proj Vector}
   RigVMVariableNode(
    Value=&RigVMFunction_MathVectorAdd(B=RigVMVariableNode() {variable=Return_Proj Vector}.Value, A=RigVMVariableNode() {variable=Return_Plane Transform}.Value.Translation) {template="Add::Execute(in A,in B,out Result)"}.Result
    ExecuteContext=@RigVMFunction_ControlFlowBranch.Completed
   ) {variable=Return_Proj Point}
  ]}
 New Function_ContainedGraph {nodes [
   n1 = RigVMFunctionEntryNode()
   n2 = &RigVMFunction_ControlFlowBranch(ExecuteContext=@n1.ExecuteContext, Condition=RigVMVariableNode() {variable=Debug Viz}.Value)
   Return = RigVMFunctionReturnNode(ExecuteContext=@n2.Completed)
   RigVMFunction_Sequence = &RigVMFunction_Sequence(ExecuteContext=@n2.True)
   n3 = RigVMVariableNode() {variable=Radius}
   Multiply = &RigVMFunction_MathDoubleMul(B=2, A=@n3.Value)
   n4 = RigVMVariableNode() {variable=Plane Transform}
   n5 = /Script/ControlRig.RigUnit_AimBoneMath(
    DebugSettings=(bEnabled=False,Scale=10,WorldOffset=(Rotation=(X=0,Y=0,Z=0,W=1),Translation=(X=0,Y=0,Z=0),Scale3D=(X=1,Y=1,Z=1)))
    Primary=(Weight=1,Axis=(X=0,Y=0,Z=1),Target=(X=1,Y=0,Z=0),Kind=Direction,Space=(Type=Bone,Name="None"))
    Secondary=(Weight=1,Axis=(X=0,Y=1,Z=0),Target=(X=1,Y=0,Z=0),Kind=Direction,Space=(Type=Bone,Name="None"))
    Weight=1
    Primary.Target=&RigVMFunction_MathTransformRotateVector(Transform=@n4.Value, Vector=RigVMVariableNode() {variable=Primary Axis}.Value) {template="Rotate Vector::Execute(in Transform,in Vector,out Result)"}.Result
    Secondary.Target=&RigVMFunction_MathTransformRotateVector(Transform=@n4.Value, Vector=RigVMVariableNode() {variable=Secondary Axis}.Value) {template="Rotate Vector::Execute(in Transform,in Vector,out Result)"}.Result
    InputTransform=RigVMVariableNode() {variable=Plane Transform}.Value
   ) {is=struct RigVMUnitNode}
   RigVMFunction_DebugRectangleNoSpace = /Script/RigVM.RigVMFunction_DebugRectangleNoSpace(
    Color=(R=0,G=1,B=1,A=1)
    DebugDrawSettings=(DepthPriority=SDPG_Foreground,Lifetime=-1)
    Thickness=0.25
    WorldOffset=(Rotation=(X=0,Y=0,Z=0,W=1),Translation=(X=0,Y=0,Z=0),Scale3D=(X=1,Y=1,Z=1))
    bEnabled=True
    Scale=@Multiply.Result
    ExecutePin=@RigVMFunction_Sequence.A
    Transform=@n5.Result
   ) {is=struct RigVMUnitNode}
   > RigVMFunction_DebugArcNoSpace = /Script/RigVM.RigVMFunction_DebugArcNoSpace(
    Color=(R=0,G=1,B=1,A=1)
    DebugDrawSettings=(DepthPriority=SDPG_Foreground,Lifetime=-1)
    Detail=64
    MaximumDegrees=360
    MinimumDegrees=0
    Thickness=0.25
    WorldOffset=(Rotation=(X=0,Y=0,Z=0,W=1),Translation=(X=0,Y=0,Z=0),Scale3D=(X=1,Y=1,Z=1))
    bEnabled=True
    Transform=@n5.Result
    Radius=@n3.Value
   ) {is=struct RigVMUnitNode}
   n6 = RigVMVariableNode() {variable=Plane Transform}
   RigVMFunction_MathVectorSetLength = &RigVMFunction_MathVectorSetLength(
    Value=&RigVMFunction_MathTransformRotateVector(Transform=@n6.Value, Vector=RigVMVariableNode() {variable=Primary Axis}.Value) {template="Rotate Vector::Execute(in Transform,in Vector,out Result)"}.Result
    Length=RigVMVariableNode() {variable=Length}.Value
   ) {template="SetLength::Execute(in Value,in Length,out Result)"}
   Add = &RigVMFunction_MathVectorAdd(A=@RigVMFunction_MathVectorSetLength.Result, B=@n6.Value.Translation) {template="Add::Execute(in A,in B,out Result)"}
   > &RigVMFunction_DebugRectangleNoSpace(
    Color=(R=0,G=1,B=1,A=1)
    DebugDrawSettings=(DepthPriority=SDPG_Foreground,Lifetime=-1)
    Scale=3.75
    Thickness=0.25
    Transform=(Rotation=(X=0,Y=0,Z=0,W=1),Translation=(X=0,Y=0,Z=0),Scale3D=(X=1,Y=1,Z=1))
    WorldOffset=(Rotation=(X=0,Y=0,Z=0,W=1),Translation=(X=0,Y=0,Z=0),Scale3D=(X=1,Y=1,Z=1))
    bEnabled=True
    Transform.Translation=@Add.Result
    Transform.Scale3D=@n6.Value.Scale3D
    Transform.Rotation=@n5.Result.Rotation
   )
   /Script/RigVM.RigVMFunction_DebugTransformMutableNoSpace(
    Color=(R=1,G=1,B=1,A=1)
    DebugDrawSettings=(DepthPriority=SDPG_Foreground,Lifetime=-1)
    Mode=Point
    Scale=5
    Thickness=0
    Transform=(Rotation=(X=0,Y=0,Z=0,W=1),Translation=(X=0,Y=0,Z=0),Scale3D=(X=1,Y=1,Z=1))
    WorldOffset=(Rotation=(X=0,Y=0,Z=0,W=1),Translation=(X=0,Y=0,Z=0),Scale3D=(X=1,Y=1,Z=1))
    bEnabled=True
    Transform.Translation=RigVMVariableNode() {variable=Proj Point}.Value
    ExecutePin=@RigVMFunction_Sequence.B
   ) {is=struct RigVMUnitNode}
   > /Script/RigVM.RigVMFunction_DebugLineStripNoSpace(
    Color=(R=1,G=1,B=1,A=1)
    DebugDrawSettings=(DepthPriority=SDPG_Foreground,Lifetime=-1)
    Points=((X=0,Y=0,Z=0),(X=0,Y=0,Z=0))
    Thickness=0.25
    WorldOffset=(Rotation=(X=0,Y=0,Z=0,W=1),Translation=(X=0,Y=0,Z=0),Scale3D=(X=1,Y=1,Z=1))
    bEnabled=True
    Points.0=RigVMVariableNode() {variable=Plane Transform}.Value.Translation
    Points.1=RigVMVariableNode() {variable=Proj Point}.Value
   ) {is=struct RigVMUnitNode}
   > &RigVMFunction_DebugLineStripNoSpace(
    Color=(R=1,G=1,B=1,A=1)
    DebugDrawSettings=(DepthPriority=SDPG_Foreground,Lifetime=-1)
    Points=((X=0,Y=0,Z=0),(X=0,Y=0,Z=0))
    Thickness=0.25
    WorldOffset=(Rotation=(X=0,Y=0,Z=0,W=1),Translation=(X=0,Y=0,Z=0),Scale3D=(X=1,Y=1,Z=1))
    bEnabled=True
    Points.0=RigVMVariableNode() {variable=Plane Transform}.Value.Translation
    Points.1=RigVMVariableNode() {variable=Aim Tip}.Value
   )
   n7 = RigVMVariableNode() {variable=Proj Point}
   Subtract = &RigVMFunction_MathVectorSub(A=RigVMVariableNode() {variable=Aim Tip}.Value, B=@n7.Value) {template="Subtract::Execute(in A,in B,out Result)"}
   n8 = RigVMVariableNode() {variable=Hypoyhenuse}
   > &RigVMFunction_DebugLineStripNoSpace(
    Color=(R=0,G=1,B=1,A=1)
    DebugDrawSettings=(DepthPriority=SDPG_Foreground,Lifetime=-1)
    Points=((X=0,Y=0,Z=0),(X=0,Y=0,Z=0))
    Thickness=0.375
    WorldOffset=(Rotation=(X=0,Y=0,Z=0,W=1),Translation=(X=0,Y=0,Z=0),Scale3D=(X=1,Y=1,Z=1))
    bEnabled=True
    Points.1=&RigVMFunction_MathVectorAdd(
     B=&RigVMFunction_MathVectorSetLength(Value=@Subtract.Result, Length=@n8.Value) {template="SetLength::Execute(in Value,in Length,out Result)"}.Result
     A=@n7.Value
    ) {template="Add::Execute(in A,in B,out Result)"}.Result
    Points.0=@n7.Value
   )
   > RigVMFunction_DebugTransformMutableNoSpace = &RigVMFunction_DebugTransformMutableNoSpace(
    Color=(R=1,G=0,B=0,A=1)
    DebugDrawSettings=(DepthPriority=SDPG_Foreground,Lifetime=-1)
    Mode=Axes
    Scale=10
    Thickness=0
    WorldOffset=(Rotation=(X=0,Y=0,Z=0,W=1),Translation=(X=0,Y=0,Z=0),Scale3D=(X=1,Y=1,Z=1))
    bEnabled=True
    Transform=RigVMVariableNode() {variable=Hinge Transform}.Value
   )
  ]}
 New Function_ContainedGraph {nodes [
   RigVMFunctionEntryNode()
   > RigVMFunction_Sequence = RigVMAggregateNode()
   RigVMVariableNode(ExecuteContext=@RigVMFunction_Sequence.A, Value=RigVMVariableNode() {variable=Plane Transform}.Value) {variable=Return_Border Transform}
   Subtract = &RigVMFunction_MathVectorSub(A=RigVMVariableNode() {variable=Aim Tip}.Value, B=RigVMVariableNode() {variable=Proj Point}.Value) {template="Subtract::Execute(in A,in B,out Result)"}
   > RigVMVariableNode(Value=@Subtract.Result) {variable=Hinge Vector}
   n1 = RigVMVariableNode() {variable=Plane Normal}
   n2 = RigVMVariableNode() {variable=Aim Vector}
   > RigVMVariableNode(Value=/Script/RigVM.RigVMFunction_MathDoubleDeg(Value=/Script/RigVM.RigVMFunction_MathVectorAngle(A=@n1.Value, B=@n2.Value) {is=struct RigVMUnitNode, template="AngleBetween::Execute(in A,in B,out Result)"}.Result) {is=struct RigVMUnitNode, template="Degrees::Execute(in Value,out Result)"}.Result) {variable=Aim Normal Degree}
   RigVMFunction_MathVectorSetLength = &RigVMFunction_MathVectorSetLength(Value=RigVMVariableNode() {variable=Plane Normal}.Value, Length=RigVMVariableNode() {variable=Length}.Value) {template="SetLength::Execute(in Value,in Length,out Result)"}
   Add = &RigVMFunction_MathVectorAdd(B=@RigVMFunction_MathVectorSetLength.Result, A=RigVMVariableNode() {variable=Plane Transform}.Value.Translation) {template="Add::Execute(in A,in B,out Result)"}
   > RigVMVariableNode(Value=@Add.Result) {variable=Compare Aim To Pos}
   Return = RigVMFunctionReturnNode(
    ExecuteContext=@RigVMFunction_Sequence.C
    Border Compare Pos=RigVMVariableNode() {variable=Return_Border Compare Pos}.Value
    Border Transform=RigVMVariableNode() {variable=Return_Border Transform}.Value
   )
   RigVMFunction_ControlFlowBranch = &RigVMFunction_ControlFlowBranch(
    Condition=RigVMDispatchNode(B=0, A=RigVMVariableNode() {variable=Aim Normal Degree}.Value) {template="DISPATCH_RigVMDispatch_CoreEquals(in A,in B,out Result)"}.Result
    ExecuteContext=@RigVMFunction_Sequence.B
   )
   n3 = RigVMVariableNode() {variable=Aim Normal Degree}
   Greater = &RigVMFunction_MathDoubleGreater(B=0, A=@n3.Value) {template="Greater::Execute(in A,in B,out Result)"}
   Less = /Script/RigVM.RigVMFunction_MathDoubleLess(B=90, A=@n3.Value) {is=struct RigVMUnitNode, template="Less::Execute(in A,in B,out Result)"}
   RigVMFunction_MathBoolAnd = &RigVMFunction_MathBoolAnd(A=@Greater.Result, B=@Less.Result)
   n4 = &RigVMFunction_ControlFlowBranch(ExecuteContext=@RigVMFunction_ControlFlowBranch.False, Condition=@RigVMFunction_MathBoolAnd.Result)
   n5 = &RigVMFunction_ControlFlowBranch(
    Condition=RigVMDispatchNode(B=90, A=RigVMVariableNode() {variable=Aim Normal Degree}.Value) {template="DISPATCH_RigVMDispatch_CoreEquals(in A,in B,out Result)"}.Result
    ExecuteContext=@n4.False
   )
   n6 = RigVMVariableNode() {variable=Aim Normal Degree}
   n7 = &RigVMFunction_ControlFlowBranch(
    ExecuteContext=@n5.False
    Condition=&RigVMFunction_MathBoolAnd(
     A=&RigVMFunction_MathDoubleGreater(B=90, A=@n6.Value) {template="Greater::Execute(in A,in B,out Result)"}.Result
     B=&RigVMFunction_MathDoubleLess(B=180, A=@n6.Value) {template="Less::Execute(in A,in B,out Result)"}.Result
    ).Result
   )
   n8 = &RigVMFunction_ControlFlowBranch(
    Condition=RigVMDispatchNode(B=180, A=RigVMVariableNode() {variable=Aim Normal Degree}.Value) {template="DISPATCH_RigVMDispatch_CoreEquals(in A,in B,out Result)"}.Result
    ExecuteContext=@n7.False
   )
   n9 = RigVMVariableNode() {variable=Plane Transform}
   RigVMVariableNode(ExecuteContext=@n8.True, Value=@n9.Value) {variable=Return_Border Transform}
   > RigVMVariableNode(Value=@n9.Value) {variable=Return_Border Compare Pos}
   n10 = &RigUnit_AimBoneMath(
    DebugSettings=(bEnabled=False,Scale=10,WorldOffset=(Rotation=(X=0,Y=0,Z=0,W=1),Translation=(X=0,Y=0,Z=0),Scale3D=(X=1,Y=1,Z=1)))
    InputTransform=(Rotation=(X=0,Y=0,Z=0,W=1),Translation=(X=0,Y=0,Z=0),Scale3D=(X=1,Y=1,Z=1))
    Primary=(Weight=1,Axis=(X=1,Y=0,Z=0),Target=(X=1,Y=0,Z=0),Kind=Direction,Space=(Type=Bone,Name="None"))
    Secondary=(Weight=1,Axis=(X=0,Y=0,Z=1),Target=(X=1,Y=0,Z=0),Kind=Location,Space=(Type=Bone,Name="None"))
    Weight=1
    Primary.Axis=RigVMVariableNode() {variable=Primary Axis}.Value
    Primary.Target=RigVMVariableNode() {variable=Hinge Vector}.Value
    Secondary.Axis=RigVMVariableNode() {variable=Secondary Axis}.Value
    InputTransform.Translation=RigVMVariableNode() {variable=Proj Point}.Value
    Secondary.Target=RigVMVariableNode() {variable=Plane Transform}.Value.Translation
   )
   RigVMVariableNode(Value=@n10.Result, ExecuteContext=@n7.True) {variable=Return_Border Transform}
   n11 = &RigUnit_AimBoneMath(
    DebugSettings=(bEnabled=False,Scale=10,WorldOffset=(Rotation=(X=0,Y=0,Z=0,W=1),Translation=(X=0,Y=0,Z=0),Scale3D=(X=1,Y=1,Z=1)))
    InputTransform=(Rotation=(X=0,Y=0,Z=0,W=1),Translation=(X=0,Y=0,Z=0),Scale3D=(X=1,Y=1,Z=1))
    Primary=(Weight=1,Axis=(X=1,Y=0,Z=0),Target=(X=1,Y=0,Z=0),Kind=Location,Space=(Type=Bone,Name="None"))
    Secondary=(Weight=1,Axis=(X=0,Y=0,Z=1),Target=(X=1,Y=0,Z=0),Kind=Location,Space=(Type=Bone,Name="None"))
    Weight=1
    Secondary.Axis=/Script/RigVM.RigVMFunction_MathVectorScale(Factor=-1, Value=RigVMVariableNode() {variable=Secondary Axis}.Value) {is=struct RigVMUnitNode, template="Scale::Execute(in Value,in Factor,out Result)"}.Result
    Primary.Axis=RigVMVariableNode() {variable=Primary Axis}.Value
    Primary.Target=RigVMVariableNode() {variable=Compare Aim To Pos}.Value
    InputTransform.Translation=RigVMVariableNode() {variable=Proj Point}.Value
    Secondary.Target=RigVMVariableNode() {variable=Plane Transform}.Value.Translation
   )
   > RigVMVariableNode(Value=@n11.Result) {variable=Return_Border Compare Pos}
   n12 = &RigUnit_AimBoneMath(
    DebugSettings=(bEnabled=False,Scale=10,WorldOffset=(Rotation=(X=0,Y=0,Z=0,W=1),Translation=(X=0,Y=0,Z=0),Scale3D=(X=1,Y=1,Z=1)))
    InputTransform=(Rotation=(X=0,Y=0,Z=0,W=1),Translation=(X=0,Y=0,Z=0),Scale3D=(X=1,Y=1,Z=1))
    Primary=(Weight=1,Axis=(X=1,Y=0,Z=0),Target=(X=1,Y=0,Z=0),Kind=Direction,Space=(Type=Bone,Name="None"))
    Secondary=(Weight=1,Axis=(X=0,Y=0,Z=1),Target=(X=1,Y=0,Z=0),Kind=Direction,Space=(Type=Bone,Name="None"))
    Weight=1
    Secondary.Target=&RigVMFunction_MathVectorScale(Factor=-1, Value=RigVMVariableNode() {variable=Plane Normal}.Value) {template="Scale::Execute(in Value,in Factor,out Result)"}.Result
    Secondary.Axis=&RigVMFunction_MathVectorScale(Factor=-1, Value=RigVMVariableNode() {variable=Secondary Axis}.Value) {template="Scale::Execute(in Value,in Factor,out Result)"}.Result
    Primary.Axis=RigVMVariableNode() {variable=Primary Axis}.Value
    Primary.Target=RigVMVariableNode() {variable=Hinge Vector}.Value
    InputTransform.Translation=RigVMVariableNode() {variable=Proj Point}.Value
   )
   RigVMVariableNode(Value=@n12.Result, ExecuteContext=@n5.True) {variable=Return_Border Transform}
   n13 = &RigUnit_AimBoneMath(
    DebugSettings=(bEnabled=False,Scale=10,WorldOffset=(Rotation=(X=0,Y=0,Z=0,W=1),Translation=(X=0,Y=0,Z=0),Scale3D=(X=1,Y=1,Z=1)))
    InputTransform=(Rotation=(X=0,Y=0,Z=0,W=1),Translation=(X=0,Y=0,Z=0),Scale3D=(X=1,Y=1,Z=1))
    Primary=(Weight=1,Axis=(X=1,Y=0,Z=0),Target=(X=1,Y=0,Z=0),Kind=Location,Space=(Type=Bone,Name="None"))
    Secondary=(Weight=1,Axis=(X=0,Y=0,Z=1),Target=(X=1,Y=0,Z=0),Kind=Location,Space=(Type=Bone,Name="None"))
    Weight=1
    Secondary.Axis=&RigVMFunction_MathVectorScale(Factor=-1, Value=RigVMVariableNode() {variable=Secondary Axis}.Value) {template="Scale::Execute(in Value,in Factor,out Result)"}.Result
    Primary.Axis=RigVMVariableNode() {variable=Primary Axis}.Value
    Primary.Target=RigVMVariableNode() {variable=Compare Aim To Pos}.Value
    InputTransform.Translation=RigVMVariableNode() {variable=Proj Point}.Value
    Secondary.Target=RigVMVariableNode() {variable=Plane Transform}.Value.Translation
   )
   > RigVMVariableNode(Value=@n13.Result) {variable=Return_Border Compare Pos}
   n14 = &RigUnit_AimBoneMath(
    DebugSettings=(bEnabled=False,Scale=10,WorldOffset=(Rotation=(X=0,Y=0,Z=0,W=1),Translation=(X=0,Y=0,Z=0),Scale3D=(X=1,Y=1,Z=1)))
    InputTransform=(Rotation=(X=0,Y=0,Z=0,W=1),Translation=(X=0,Y=0,Z=0),Scale3D=(X=1,Y=1,Z=1))
    Primary=(Weight=1,Axis=(X=1,Y=0,Z=0),Target=(X=1,Y=0,Z=0),Kind=Direction,Space=(Type=Bone,Name="None"))
    Secondary=(Weight=1,Axis=(X=0,Y=0,Z=1),Target=(X=1,Y=0,Z=0),Kind=Location,Space=(Type=Bone,Name="None"))
    Weight=1
    Secondary.Axis=&RigVMFunction_MathVectorScale(Factor=-1, Value=RigVMVariableNode() {variable=Secondary Axis}.Value) {template="Scale::Execute(in Value,in Factor,out Result)"}.Result
    Primary.Axis=RigVMVariableNode() {variable=Primary Axis}.Value
    Primary.Target=RigVMVariableNode() {variable=Hinge Vector}.Value
    InputTransform.Translation=RigVMVariableNode() {variable=Proj Point}.Value
    Secondary.Target=RigVMVariableNode() {variable=Plane Transform}.Value.Translation
   )
   RigVMVariableNode(Value=@n14.Result, ExecuteContext=@n4.True) {variable=Return_Border Transform}
   n15 = &RigUnit_AimBoneMath(
    DebugSettings=(bEnabled=False,Scale=10,WorldOffset=(Rotation=(X=0,Y=0,Z=0,W=1),Translation=(X=0,Y=0,Z=0),Scale3D=(X=1,Y=1,Z=1)))
    InputTransform=(Rotation=(X=0,Y=0,Z=0,W=1),Translation=(X=0,Y=0,Z=0),Scale3D=(X=1,Y=1,Z=1))
    Primary=(Weight=1,Axis=(X=1,Y=0,Z=0),Target=(X=1,Y=0,Z=0),Kind=Location,Space=(Type=Bone,Name="None"))
    Secondary=(Weight=1,Axis=(X=0,Y=0,Z=1),Target=(X=1,Y=0,Z=0),Kind=Location,Space=(Type=Bone,Name="None"))
    Weight=1
    Secondary.Axis=&RigVMFunction_MathVectorScale(Factor=-1, Value=RigVMVariableNode() {variable=Secondary Axis}.Value) {template="Scale::Execute(in Value,in Factor,out Result)"}.Result
    Primary.Axis=RigVMVariableNode() {variable=Primary Axis}.Value
    Primary.Target=RigVMVariableNode() {variable=Compare Aim To Pos}.Value
    InputTransform.Translation=RigVMVariableNode() {variable=Proj Point}.Value
    Secondary.Target=RigVMVariableNode() {variable=Plane Transform}.Value.Translation
   )
   > RigVMVariableNode(Value=@n15.Result) {variable=Return_Border Compare Pos}
   n16 = RigVMVariableNode() {variable=Plane Transform}
   RigVMVariableNode(ExecuteContext=@RigVMFunction_ControlFlowBranch.True, Value=@n16.Value) {variable=Return_Border Transform}
   > RigVMVariableNode(Value=@n16.Value) {variable=Return_Border Compare Pos}
  ]}
 New Function_ContainedGraph {nodes [
   n1 = RigVMFunctionEntryNode()
   RigVMFunction_MathVectorLength = &RigVMFunction_MathVectorLength(Value=RigVMVariableNode() {variable=Proj Vector}.Value) {template="Length::Execute(in Value,out Result)"}
   RigVMFunction_ControlFlowBranch = &RigVMFunction_ControlFlowBranch(
    ExecuteContext=@n1.ExecuteContext
    Condition=RigVMDispatchNode(B=0, A=@RigVMFunction_MathVectorLength.Result) {template="DISPATCH_RigVMDispatch_CoreEquals(in A,in B,out Result)"}.Result
   )
   RigVMFunction_MathVectorAngle = &RigVMFunction_MathVectorAngle(A=RigVMVariableNode() {variable=Aim Vector}.Value, B=RigVMVariableNode() {variable=Proj Vector}.Value) {template="AngleBetween::Execute(in A,in B,out Result)"}
   Absolute = &RigVMFunction_MathDoubleAbs(Value=@RigVMFunction_MathVectorAngle.Result)
   RigVMVariableNode(Value=@Absolute.Result, ExecuteContext=@RigVMFunction_ControlFlowBranch.False) {variable=" Rads Aim Proj"}
   n2 = &RigVMFunction_MathDoubleDeg(Value=@Absolute.Result) {template="Degrees::Execute(in Value,out Result)"}
   > n3 = RigVMVariableNode(Value=@n2.Result) {variable=Degs Aim Proj}
   n4 = &RigVMFunction_ControlFlowBranch(
    Condition=RigVMDispatchNode(B=0, A=@n2.Result) {template="DISPATCH_RigVMDispatch_CoreEquals(in A,in B,out Result)"}.Result
    ExecuteContext=@n3.ExecuteContext
   )
   RigVMFunction_Sequence = RigVMAggregateNode(ExecuteContext=@n4.False)
   n5 = &RigVMFunction_MathDoubleSub(
    A=180
    B=&RigVMFunction_MathDoubleAdd(B=RigVMVariableNode() {variable=Degs Aim Hyp}.Value, A=RigVMVariableNode() {variable=Degs Aim Proj}.Value).Result
   )
   RigVMVariableNode(ExecuteContext=@RigVMFunction_Sequence.B, Value=@n5.Result) {variable=Degs Proj Hyp}
   > RigVMVariableNode(Value=/Script/RigVM.RigVMFunction_MathDoubleRad(Value=@n5.Result) {is=struct RigVMUnitNode, template="Radians::Execute(in Value,out Result)"}.Result) {variable=Rads Proj Hyp}
   n6 = /Script/RigVM.RigVMFunction_MathDoubleSin(Value=RigVMVariableNode() {variable=" Rads Aim Proj"}.Value) {is=struct RigVMUnitNode, template="Sin::Execute(in Value,out Result)"}
   n7 = RigVMVariableNode() {variable=Radius}
   n8 = /Script/RigVM.RigVMFunction_MathDoubleAsin(Value=&RigVMFunction_MathDoubleDiv(A=&RigVMFunction_MathDoubleMul(A=@n6.Result, B=@n7.Value).Result, B=RigVMVariableNode() {variable=Hypothenuse}.Value).Result) {is=struct RigVMUnitNode, template="Asin::Execute(in Value,out Result)"}
   RigVMVariableNode(Value=@n8.Result, ExecuteContext=@RigVMFunction_Sequence.A) {variable=Rads Aim Hyp}
   > RigVMVariableNode(Value=&RigVMFunction_MathDoubleDeg(Value=@n8.Result) {template="Degrees::Execute(in Value,out Result)"}.Result) {variable=Degs Aim Hyp}
   Sin = &RigVMFunction_MathDoubleSin(Value=RigVMVariableNode() {variable=Rads Proj Hyp}.Value) {template="Sin::Execute(in Value,out Result)"}
   n9 = RigVMVariableNode() {variable=Hypothenuse}
   n10 = RigVMVariableNode() {variable=" Rads Aim Proj"}
   RigVMVariableNode(
    Value=&RigVMFunction_MathDoubleDiv(
     A=&RigVMFunction_MathDoubleMul(B=@Sin.Result, A=@n9.Value).Result
     B=&RigVMFunction_MathDoubleSin(Value=@n10.Value) {template="Sin::Execute(in Value,out Result)"}.Result
    ).Result
    ExecuteContext=@RigVMFunction_Sequence.C
   ) {variable=Return Scaled Length}
   RigVMVariableNode(
    Value=&RigVMFunction_MathDoubleAdd(A=RigVMVariableNode() {variable=Length}.Value, B=RigVMVariableNode() {variable=Radius}.Value).Result
    ExecuteContext=@n4.True
   ) {variable=Return Scaled Length}
   > RigVMVariableNode(Value=0) {variable=Rads Aim Hyp}
   > RigVMVariableNode(Value=0) {variable=Degs Aim Hyp}
   Pi = /Script/RigVM.RigVMFunction_MathDoubleConstPi() {is=struct RigVMUnitNode, template="Pi::Execute(out Value)"}
   > RigVMVariableNode(Value=@Pi.Value) {variable=Rads Proj Hyp}
   > RigVMVariableNode(Value=180) {variable=Degs Proj Hyp}
   n11 = RigVMVariableNode() {variable=Return Scaled Length}
   RigVMFunction_MathVectorSetLength = &RigVMFunction_MathVectorSetLength(Length=@n11.Value, Value=RigVMVariableNode() {variable=Aim Vector}.Value) {template="SetLength::Execute(in Value,in Length,out Result)"}
   Return = RigVMFunctionReturnNode(
    ExecuteContext=@RigVMFunction_ControlFlowBranch.Completed
    Scaled Length=@n11.Value
    Scaled Aim Vector=@RigVMFunction_MathVectorSetLength.Result
   )
   RigVMVariableNode(Value=RigVMVariableNode() {variable=Length}.Value, ExecuteContext=@RigVMFunction_ControlFlowBranch.True) {variable=Return Scaled Length}
  ]}
 New Function_ContainedGraph {nodes [
   RigVMFunctionEntryNode()
   Multiply = &RigVMFunction_MathTransformMul(
    A=&RigVMFunction_MathTransformMakeRelative(Global=RigVMVariableNode() {variable=Input Transform}.Value, Parent=RigVMVariableNode() {variable=Rest Transform}.Value).Local
    B=RigVMVariableNode() {variable=Active Transform}.Value
   ) {template="Multiply::Execute(in A,in B,out Result)"}
   > Return = RigVMFunctionReturnNode(Result Transform=@Multiply.Result)
  ]}
 New Function_ContainedGraph {nodes [
   RigVMFunctionEntryNode()
   > Pivot Update = RigVMFunctionReferenceNode(
    Input Transform=RigVMVariableNode() {variable=Input Transform}.Value
    Rest Transform=RigVMVariableNode() {variable=Rest Transform}.Value
    Active Transform=RigVMVariableNode() {variable=Active Transform}.Value
   )
   n1 = RigVMVariableNode() {variable=Aim To Element}
   n2 = &RigUnit_AimBoneMath(
    DebugSettings=(bEnabled=False,Scale=10,WorldOffset=(Rotation=(X=0,Y=0,Z=0,W=1),Translation=(X=0,Y=0,Z=0),Scale3D=(X=1,Y=1,Z=1)))
    Primary=(Weight=1,Axis=(X=1,Y=0,Z=0),Target=(X=1,Y=0,Z=0),Kind=Location,Space=(Type=Bone,Name="None"))
    Secondary=(Weight=1,Axis=(X=0,Y=0,Z=1),Target=(X=1,Y=0,Z=0),Kind=Direction,Space=(Type=Bone,Name="None"))
    Weight=1
    Secondary.Target=&RigVMFunction_MathTransformRotateVector(
     Transform=&RigUnit_GetTransform(bInitial=False, Item=@n1.Value).Transform
     Vector=RigVMVariableNode() {variable=Secondary Axis}.Value
    ) {template="Rotate Vector::Execute(in Transform,in Vector,out Result)"}.Result
    Primary.Target=RigVMVariableNode() {variable=Aim Tip}.Value
    Primary.Axis=RigVMVariableNode() {variable=Primary Axis}.Value
    Secondary.Axis=RigVMVariableNode() {variable=Secondary Axis}.Value
    InputTransform=@"Pivot Update".Result Transform
   )
   > Return = RigVMFunctionReturnNode(Result Transform=@n2.Result)
  ]}
 New Function_ContainedGraph {nodes [
   n1 = RigVMFunctionEntryNode()
   Multiply = &RigVMFunction_MathTransformMul(
    B=&RigUnit_GetTransform(bInitial=False, Item=RigVMVariableNode() {variable=Driver}.Value).Transform
    A=&RigVMFunction_MathTransformMakeRelative(
     Parent=&RigUnit_GetTransform(bInitial=true, Item=@n1.Driver).Transform
     Global=&RigUnit_GetTransform(bInitial=true, Item=@n1.Driven).Transform
    ).Local
   ) {template="Multiply::Execute(in A,in B,out Result)"}
   > &RigUnit_SetTransform(Weight=1, bInitial=False, bPropagateToChildren=false, Value=@Multiply.Result, Item=RigVMVariableNode() {variable=Driven}.Value)
   > Return = RigVMFunctionReturnNode(Result Transform=@Multiply.Result)
  ]}
 RigVMFunctionLibrary {nodes [
   Debug Display = RigVMCollapseNode(
    Aim Tip=(X=0,Y=0,Z=0)
    Aim Vector=(X=0,Y=0,Z=0)
    ExecuteContext=()
    Hinge Transform=(Rotation=(X=0,Y=0,Z=0,W=1),Translation=(X=0,Y=0,Z=0),Scale3D=(X=1,Y=1,Z=1))
    Plane Transform=(Rotation=(X=0,Y=0,Z=0,W=1),Translation=(X=0,Y=0,Z=0),Scale3D=(X=1,Y=1,Z=1))
    Primary Axis=(X=0,Y=0,Z=0)
    Proj Point=(X=0,Y=0,Z=0)
    Proj Vector=(X=0,Y=0,Z=0)
    Secondary Axis=(X=0,Y=0,Z=0)
   )
   Drive With Offset = RigVMCollapseNode(Driven=(Type=None,Name="None"), Driver=(Type=None,Name="None"), ExecuteContext=())
   Get Border Transform = RigVMCollapseNode(
    Aim Tip=(X=0,Y=0,Z=0)
    Aim Vector=(X=0,Y=0,Z=0)
    ExecuteContext=()
    Plane Normal=(X=0,Y=0,Z=0)
    Plane Transform=(Rotation=(X=0,Y=0,Z=0,W=1),Translation=(X=0,Y=0,Z=0),Scale3D=(X=1,Y=1,Z=1))
    Primary Axis=(X=0,Y=0,Z=0)
    Proj Point=(X=0,Y=0,Z=0)
    Secondary Axis=(X=0,Y=0,Z=0)
   )
   Get Final Base Transform = RigVMCollapseNode(
    Active Transform=(Rotation=(X=0,Y=0,Z=0,W=1),Translation=(X=0,Y=0,Z=0),Scale3D=(X=1,Y=1,Z=1))
    Aim Tip=(X=0,Y=0,Z=0)
    Aim To Element=(Type=None,Name="None")
    ExecuteContext=()
    Input Transform=(Rotation=(X=0,Y=0,Z=0,W=1),Translation=(X=0,Y=0,Z=0),Scale3D=(X=1,Y=1,Z=1))
    Primary Axis=(X=0,Y=0,Z=0)
    Rest Transform=(Rotation=(X=0,Y=0,Z=0,W=1),Translation=(X=0,Y=0,Z=0),Scale3D=(X=1,Y=1,Z=1))
    Secondary Axis=(X=0,Y=0,Z=0)
   )
   Get Projection = RigVMCollapseNode(Aim To Element=(Type=None,Name="None"), ExecuteContext=(), Plane Element=(Type=None,Name="None"), Primary Axis=(X=0,Y=0,Z=0))
   Get Scaled Aim Vector = RigVMCollapseNode(Aim Vector=(X=0,Y=0,Z=0), ExecuteContext=(), Proj Vector=(X=0,Y=0,Z=0))
   Pivot Update = RigVMCollapseNode(
    Active Transform=(Rotation=(X=0,Y=0,Z=0,W=1),Translation=(X=0,Y=0,Z=0),Scale3D=(X=1,Y=1,Z=1))
    ExecuteContext=()
    Input Transform=(Rotation=(X=0,Y=0,Z=0,W=1),Translation=(X=0,Y=0,Z=0),Scale3D=(X=1,Y=1,Z=1))
    Rest Transform=(Rotation=(X=0,Y=0,Z=0,W=1),Translation=(X=0,Y=0,Z=0),Scale3D=(X=1,Y=1,Z=1))
   )
  ]}
 RigVMFunction_Sequence_1_1_ContainedGraph {nodes [
   n1 = RigVMFunctionEntryNode()
   n2 = &RigVMFunction_Sequence(ExecuteContext=@n1.ExecuteContext)
   n3 = &RigVMFunction_Sequence(ExecuteContext=@n2.B)
   n4 = &RigVMFunction_Sequence(ExecuteContext=@n3.B)
   n5 = &RigVMFunction_Sequence(ExecuteContext=@n4.B)
   Return = RigVMFunctionReturnNode(A=@n2.A, B=@n3.A, C=@n4.A, D=@n5.A, E=@n5.B)
  ]}
 RigVMFunction_Sequence_1_ContainedGraph {nodes [
   n1 = RigVMFunctionEntryNode()
   RigVMFunction_Sequence = &RigVMFunction_Sequence(ExecuteContext=@n1.ExecuteContext)
   n2 = &RigVMFunction_Sequence(ExecuteContext=@RigVMFunction_Sequence.B)
   n3 = &RigVMFunction_Sequence(ExecuteContext=@n2.B)
   n4 = &RigVMFunction_Sequence(ExecuteContext=@n3.B)
   Return = RigVMFunctionReturnNode(A=@RigVMFunction_Sequence.A, B=@n2.A, C=@n3.A, D=@n4.A, E=@n4.B)
  ]}
 RigVMFunction_Sequence_1_ContainedGraph {nodes [
   n1 = RigVMFunctionEntryNode()
   RigVMFunction_Sequence = &RigVMFunction_Sequence(ExecuteContext=@n1.ExecuteContext)
   n2 = &RigVMFunction_Sequence(ExecuteContext=@RigVMFunction_Sequence.B)
   Return = RigVMFunctionReturnNode(A=@RigVMFunction_Sequence.A, B=@n2.A, C=@n2.B)
  ]}
 RigVMFunction_Sequence_1_ContainedGraph {nodes [
   n1 = RigVMFunctionEntryNode()
   RigVMFunction_Sequence = &RigVMFunction_Sequence(ExecuteContext=@n1.ExecuteContext)
   n2 = &RigVMFunction_Sequence(ExecuteContext=@RigVMFunction_Sequence.B)
   n3 = &RigVMFunction_Sequence(ExecuteContext=@n2.B)
   n4 = &RigVMFunction_Sequence(ExecuteContext=@n3.B)
   Return = RigVMFunctionReturnNode(A=@RigVMFunction_Sequence.A, B=@n2.A, C=@n3.A, D=@n4.A, E=@n4.B)
  ]}
 RigVMFunction_Sequence_2_ContainedGraph {nodes [
   n1 = RigVMFunctionEntryNode()
   RigVMFunction_Sequence = &RigVMFunction_Sequence(ExecuteContext=@n1.ExecuteContext)
   n2 = &RigVMFunction_Sequence(ExecuteContext=@RigVMFunction_Sequence.B)
   n3 = &RigVMFunction_Sequence(ExecuteContext=@n2.B)
   Return = RigVMFunctionReturnNode(A=@RigVMFunction_Sequence.A, B=@n2.A, C=@n3.A, D=@n3.B)
  ]}
 RigVMModel Forward Event Graph {nodes [
   n1 = &RigUnit_BeginExecution()
   RigVMFunction_Sequence = RigVMAggregateNode(ExecuteContext=@n1.ExecutePin)
   Get Projection_1 = RigVMFunctionReferenceNode(
    Plane Element=RigVMVariableNode() {variable=Plane Element}.Value
    Aim To Element=RigVMVariableNode() {variable=Aim To Element}.Value
    Primary Axis=RigVMVariableNode() {variable=Primary Axis}.Value
    Radius=RigVMVariableNode() {variable=Radius}.Value
    Hypothenuse=RigVMVariableNode() {variable=Hypotenuse}.Value
    Length=RigVMVariableNode() {variable=Length}.Value
    ExecuteContext=@RigVMFunction_Sequence.B
   )
   > RigVMVariableNode(Value=@"Get Projection_1".Proj Vector) {variable=Proj Vector}
   > RigVMVariableNode(Value=@"Get Projection_1".Proj Point) {variable=Proj Point}
   > RigVMVariableNode(Value=@"Get Projection_1".Aim Vector) {variable=Aim Vector}
   > RigVMVariableNode(Value=@"Get Projection_1".Plane Normal) {variable=Plane Normal}
   > RigVMVariableNode(Value=@"Get Projection_1".Aim Tip) {variable=Aim Tip}
   > Get Border Transform = RigVMFunctionReferenceNode(
    Aim Tip=RigVMVariableNode() {variable=Aim Tip}.Value
    Proj Point=RigVMVariableNode() {variable=Proj Point}.Value
    Primary Axis=RigVMVariableNode() {variable=Primary Axis}.Value
    Secondary Axis=RigVMVariableNode() {variable=Secondary Axis}.Value
    Length=RigVMVariableNode() {variable=Length}.Value
    Plane Transform=RigVMVariableNode() {variable=Plane Transform}.Value
    Plane Normal=RigVMVariableNode() {variable=Plane Normal}.Value
    Aim Vector=RigVMVariableNode() {variable=Aim Vector}.Value
   )
   > RigVMVariableNode(Value=@"Get Border Transform".Border Transform) {variable=Hinge Transform}
   > RigVMVariableNode(Value=@"Get Border Transform".Border Compare Pos) {variable=Hinge Rest Transform}
   RigVMVariableNode(
    Value=&RigUnit_GetTransform(bInitial=False, Item=RigVMVariableNode() {variable=Plane Element}.Value).Transform
    ExecuteContext=@RigVMFunction_Sequence.A
   ) {variable=Plane Transform}
   > RigVMVariableNode(Value=&RigUnit_GetTransform(bInitial=False, Item=RigVMVariableNode() {variable=Aim To Element}.Value).Transform) {variable=Aim To Transform}
   Debug Display_1 = RigVMFunctionReferenceNode(
    Plane Transform=RigVMVariableNode() {variable=Plane Transform}.Value
    Radius=RigVMVariableNode() {variable=Radius}.Value
    Length=RigVMVariableNode() {variable=Length}.Value
    Primary Axis=RigVMVariableNode() {variable=Primary Axis}.Value
    Secondary Axis=RigVMVariableNode() {variable=Secondary Axis}.Value
    Debug Viz=RigVMVariableNode() {variable=Debug Viz}.Value
    Aim Tip=RigVMVariableNode() {variable=Aim Tip}.Value
    Proj Vector=RigVMVariableNode() {variable=Proj Vector}.Value
    Hinge Transform=RigVMVariableNode() {variable=Hinge Transform}.Value
    Proj Point=RigVMVariableNode() {variable=Proj Point}.Value
    Aim Vector=RigVMVariableNode() {variable=Aim Vector}.Value
    Hypoyhenuse=RigVMVariableNode() {variable=Hypotenuse}.Value
    ExecuteContext=@RigVMFunction_Sequence.D
   )
   Get Final Base Transform = RigVMFunctionReferenceNode(
    Input Transform=RigVMVariableNode() {variable=Plane Transform}.Value
    Active Transform=RigVMVariableNode() {variable=Hinge Transform}.Value
    Rest Transform=RigVMVariableNode() {variable=Hinge Rest Transform}.Value
    Aim Tip=RigVMVariableNode() {variable=Aim Tip}.Value
    Primary Axis=RigVMVariableNode() {variable=Primary Axis}.Value
    Secondary Axis=RigVMVariableNode() {variable=Secondary Axis}.Value
    Aim To Element=RigVMVariableNode() {variable=Aim To Element}.Value
    ExecuteContext=@RigVMFunction_Sequence.C
   )
   > RigVMVariableNode(Value=@"Get Final Base Transform".Result Transform) {variable=Result Transform}
   > &RigUnit_SetTransform(
    Weight=1
    bInitial=False
    Item=RigVMVariableNode() {variable=Base Element}.Value
    Value=RigVMVariableNode() {variable=Result Transform}.Value
   )
   > &RigUnit_SetTransform(
    Weight=1
    bInitial=False
    Item=&RigUnit_ResolveConnector(Connector=(Type=Connector,Name="Driven Joint"), SkipSocket=False).Result
    Value=RigVMVariableNode() {variable=Result Transform}.Value
   )
  ]}
]
hierarchy [
 Driven Joint {type=Connector, loc=<0 0 0>, rot=<0 0 0 1>}
 Root {type=Socket, loc=<0 0 0>, rot=<0 0 0 1>}
 Aim To Parent {type=Connector, loc=<0 0 0>, rot=<0 0 0 1>}
 Parent {type=Connector, loc=<0 0 0>, rot=<0 0 0 1>}
]
variables [
 Aim Extra Length {type=double, default=10}
 Aim Tip {type=FVector, default=(X=0,Y=0,Z=0)}
 Aim To Element {type=FRigElementKey, default=(Type=None,Name="")}
 Aim To Transform {type=FTransform, default=(Rotation=(X=0,Y=0,Z=0,W=1),Translation=(X=0,Y=0,Z=0),Scale3D=(X=1,Y=1,Z=1))}
 Aim Up Vector {type=FVector, default=(X=0,Y=0,Z=0)}
 Aim Vector {type=FVector, default=(X=0,Y=0,Z=0)}
 Base Element {type=FRigElementKey, default=(Type=None,Name="")}
 Control Shape {
  type=FRigUnit_HierarchyAddControl_ShapeSettings
  default=(bVisible=True,Name="Arrow4_Thick",Color=(R=0,G=1,B=1,A=1),Transform=(Rotation=(X=-0.5,Y=-0.5,Z=0.5,W=0.5),Translation=(X=0,Y=0,Z=0),Scale3D=(X=0.75,Y=0.75,Z=0.75)))
 }
 Create Control {type=bool, default=True}
 Debug Viz {type=bool, default=True}
 Drive Tip Rotation {type=bool, default=False}
 End Element {type=FRigElementKey, default=(Type=None,Name="")}
 Hinge Rest Transform {type=FTransform, default=(Rotation=(X=0,Y=0,Z=0,W=1),Translation=(X=0,Y=0,Z=0),Scale3D=(X=1,Y=1,Z=1))}
 Hinge Transform {type=FTransform, default=(Rotation=(X=0,Y=0,Z=0,W=1),Translation=(X=0,Y=0,Z=0),Scale3D=(X=1,Y=1,Z=1))}
 Hypotenuse {type=double, default=0}
 Length {type=double, default=6.5}
 One To One Bend {type=bool, default=False}
 Plane Element {type=FRigElementKey, default=(Type=None,Name="")}
 Plane Normal {type=FVector, default=(X=0,Y=0,Z=0)}
 Plane Transform {type=FTransform, default=(Rotation=(X=0,Y=0,Z=0,W=1),Translation=(X=0,Y=0,Z=0),Scale3D=(X=1,Y=1,Z=1))}
 Primary Axis {type=FVector, default=(X=1,Y=0,Z=0)}
 Proj Point {type=FVector, default=(X=0,Y=0,Z=0)}
 Proj Vector {type=FVector, default=(X=0,Y=0,Z=0)}
 Radius {type=double, default=8}
 Result Transform {type=FTransform, default=(Rotation=(X=0,Y=0,Z=0,W=1),Translation=(X=0,Y=0,Z=0),Scale3D=(X=1,Y=1,Z=1))}
 Scale Length {type=double, default=0}
 Secondary Axis {type=FVector, default=(X=0,Y=0,Z=1)}
]
