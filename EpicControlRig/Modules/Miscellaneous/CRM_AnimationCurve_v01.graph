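# Control Rig module that exposes animation curves as float animation channels on a connected control.
# Construction spawns a channel per Curve Names entry with configurable default and limits, named from Channel Names or falling back to the curve name, and shares it with the Hosts connector's elements.
# Forward solve writes each channel times Scale Value into existing curves.
# Inverse solve divides valid curve values by Scale Value back into channels.
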ControlRigBlueprint /EpicControlRig/Modules/Miscellaneous/CRM_AnimationCurve_v01
graphs [
 CollapseNode_ContainedGraph {nodes [
   n1 = RigVMFunctionEntryNode()
   Num = RigVMDispatchNode(Array=@n1.Channel Names) {template="DISPATCH_RigVMDispatch_ArrayGetNum(in Array,out Num)"}
   Less = /Script/RigVM.RigVMFunction_MathIntLess(A=@n1.Index, B=@Num.Num) {is=struct RigVMUnitNode, template="Less::Execute(in A,in B,out Result)"}
   RigVMFunctionReturnNode(Result=RigVMDispatchNode(
     True=RigVMDispatchNode(Array=@n1.Channel Names, Index=@n1.Index) {template="DISPATCH_RigVMDispatch_ArrayGetAtIndex(in Array,in Index,out Element)"}.Element
     False=@n1.Curve Name
     Condition=@Less.Result
    ) {template="DISPATCH_RigVMDispatch_If(in Condition,in True,in False,out Result)"}.Result)
  ]}
 ConstructionGraph {nodes [
   RigUnit_PrepareForExecution = &RigUnit_PrepareForExecution()
   RigVMVariableNode(
    Value=/Script/ControlRig.RigUnit_ResolveConnector(Connector=(Type=Connector,Name="Control"), SkipSocket=False) {is=struct RigVMUnitNode}.Result
    ExecuteContext=@RigUnit_PrepareForExecution.ExecutePin
   ) {variable=Control}
   > n1 = RigVMVariableNode(Value=/Script/ControlRig.RigUnit_ResolveArrayConnector(Connector=(Type=Connector,Name="Hosts"), SkipSocket=False) {is=struct RigVMUnitNode}.Result) {variable=Hosts}
   RigVMFunction_Sequence = &RigVMFunction_Sequence(ExecuteContext=@n1.ExecuteContext)
   For_Each = RigVMDispatchNode(ExecuteContext=@RigVMFunction_Sequence.A, Array=RigVMVariableNode() {variable=Curve Names}.Value) {template="DISPATCH_RigVMDispatch_ArrayIterator(in Array,out Element,out Index,out Count,out Ratio)"}
   Get Channel Name = RigVMCollapseNode(Curve Name=@For_Each.Element, Index=@For_Each.Index, Channel Names=RigVMVariableNode() {variable=Channel Names}.Value)
   > SpawnAnimationChannel = /Script/ControlRig.RigUnit_HierarchyAddAnimationChannelFloat(
    LimitsEnabled=(Enabled=(bMinimum=True,bMaximum=True))
    Name=@"Get Channel Name".Result
    Parent=RigVMVariableNode() {variable=Control}.Value
    InitialValue=RigVMVariableNode() {variable=Default Value}.Value
    MaximumValue=RigVMVariableNode() {variable=Maximum Value}.Value
    MinimumValue=RigVMVariableNode() {variable=Minimum Value}.Value
    LimitsEnabled.Enabled.bMaximum=RigVMVariableNode() {variable=Maximum Limit Enabled}.Value
    LimitsEnabled.Enabled.bMinimum=RigVMVariableNode() {variable=Minimum Limit Enabled}.Value
   ) {
    is=struct RigVMUnitNode
    template="SpawnAnimationChannel::Execute(in InitialValue,in MinimumValue,in MaximumValue,in LimitsEnabled,in Parent,in Name,out Item)"
   }
   > RigVMDispatchNode(Element=@SpawnAnimationChannel.Item, Array=RigVMVariableNode() {variable=Channels}.Value) {template="DISPATCH_RigVMDispatch_ArrayAdd(io Array,in Element,out Index)"}
   > /Script/ControlRig.RigUnit_SetChannelHosts(Channel=@SpawnAnimationChannel.Item, Hosts=RigVMVariableNode() {variable=Hosts}.Value) {is=struct RigVMUnitNode}
  ]}
 RigVMFunctionLibrary {links [], nodes []}
 RigVMModel {nodes [
   RigUnit_BeginExecution = &RigUnit_BeginExecution()
   For_Each = RigVMDispatchNode(ExecuteContext=@RigUnit_BeginExecution.ExecutePin, Array=RigVMVariableNode() {variable=Curve Names}.Value) {template="DISPATCH_RigVMDispatch_ArrayIterator(in Array,out Element,out Index,out Count,out Ratio)"}
   RigVMFunction_ControlFlowBranch = &RigVMFunction_ControlFlowBranch(
    Condition=/Script/ControlRig.RigUnit_CurveExists(Curve=@For_Each.Element) {is=struct RigVMUnitNode}.Exists
    ExecuteContext=@For_Each.ExecuteContext
   )
   GetAnimationChannelFromItem = /Script/ControlRig.RigUnit_GetFloatAnimationChannelFromItem(
    bInitial=False
    Item=RigVMDispatchNode(Index=@For_Each.Index, Array=RigVMVariableNode() {variable=Channels}.Value) {template="DISPATCH_RigVMDispatch_ArrayGetAtIndex(in Array,in Index,out Element)"}.Element
   ) {is=struct RigVMUnitNode, template="GetAnimationChannelFromItem::Execute(out Value,in Item,in bInitial)"}
   Multiply = &RigVMFunction_MathDoubleMul(A=@GetAnimationChannelFromItem.Value, B=RigVMVariableNode() {variable=Scale Value}.Value)
   /Script/ControlRig.RigUnit_SetCurveValue(Curve=@For_Each.Element, Value=@Multiply.Result, ExecutePin=@RigVMFunction_ControlFlowBranch.True) {is=struct RigVMUnitNode}
  ]}
 RigVMModel Rig {nodes [
   n1 = /Script/ControlRig.RigUnit_InverseExecution() {is=struct RigVMUnitNode}
   For_Each = RigVMDispatchNode(ExecuteContext=@n1.ExecutePin, Array=RigVMVariableNode() {variable=Curve Names}.Value) {template="DISPATCH_RigVMDispatch_ArrayIterator(in Array,out Element,out Index,out Count,out Ratio)"}
   n2 = /Script/ControlRig.RigUnit_GetCurveValue(Curve=@For_Each.Element) {is=struct RigVMUnitNode}
   RigVMFunction_ControlFlowBranch = &RigVMFunction_ControlFlowBranch(ExecuteContext=@For_Each.ExecuteContext, Condition=@n2.Valid)
   Divide = &RigVMFunction_MathDoubleDiv(A=@n2.Value, B=RigVMVariableNode() {variable=Scale Value}.Value)
   SetAnimationChannelFromItem = /Script/ControlRig.RigUnit_SetFloatAnimationChannelFromItem(
    bInitial=False
    Item=RigVMDispatchNode(Index=@For_Each.Index, Array=RigVMVariableNode() {variable=Channels}.Value) {template="DISPATCH_RigVMDispatch_ArrayGetAtIndex(in Array,in Index,out Element)"}.Element
    Value=@Divide.Result
    ExecutePin=@RigVMFunction_ControlFlowBranch.True
   ) {is=struct RigVMUnitNode, template="SetAnimationChannelFromItem::Execute(in Value,in Item,in bInitial)"}
  ]}
]
hierarchy [
 Control {type=Connector, loc=<0 0 0>, rot=<0 0 0 1>}
 Root {type=Socket, loc=<0 0 0>, rot=<0 0 0 1>}
 disable_slopewarp {type=Curve, loc=<0 0 0>, rot=<0 0 0 1>}
 Hosts {type=Connector, loc=<0 0 0>, rot=<0 0 0 1>}
]
variables [
 Channel Names {type="TArray<FName>"}
 Channels {type="TArray<FRigElementKey>"}
 Control {type=FRigElementKey, default=(Type=Connector,Name="")}
 Curve Names {type="TArray<FName>"}
 Default Value {type=double, default=0}
 Hosts {type="TArray<FRigElementKey>"}
 Maximum Limit Enabled {type=bool, default=True}
 Maximum Value {type=double, default=1}
 Minimum Limit Enabled {type=bool, default=True}
 Minimum Value {type=double, default=0}
 Scale Value {type=double, default=1}
]
