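# Control Rig module that creates a proxy control whose rotation drives a set of other controls.
# Construction resolves the Parent, Driven Controls and Snap To connectors, then spawns the control with a Pivot Vis channel under a null at the snap target.
# While the proxy is being rotated, Forwards Solve adds its per-update Euler delta to each driven control's rotation.
# When rotation interaction ends, the proxy resets to identity, making it a relative manipulator.
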ControlRigBlueprint /EpicControlRig/Modules/Miscellaneous/CRM_Epic_ProxyControl_v01
graphs [
 RigVMFunctionLibrary {links [], nodes []}
 RigVMFunction_Sequence_2_ContainedGraph {nodes [
   n1 = RigVMFunctionEntryNode()
   n2 = &RigVMFunction_Sequence(ExecuteContext=@n1.ExecuteContext)
   n3 = &RigVMFunction_Sequence(ExecuteContext=@n2.B)
   n4 = &RigVMFunction_Sequence(ExecuteContext=@n3.B)
   RigVMFunctionReturnNode(A=@n2.A, B=@n3.A, C=@n4.A, D=@n4.B)
  ]}
 RigVMModel {nodes [
   RigUnit_BeginExecution = &RigUnit_BeginExecution()
   RigVMFunction_Sequence = &RigVMFunction_Sequence(ExecuteContext=@RigUnit_BeginExecution.ExecutePin)
   n1 = /Script/ControlRig.RigUnit_IsInteracting() {is=struct RigVMUnitNode}
   RigVMVariableNode(Value=@n1.bIsRotating, ExecuteContext=@RigVMFunction_Sequence.A) {variable=is Interacting}
   > RigVMVariableNode(Value=@n1.Items) {variable=Interacting Items}
   RigVMFunction_ControlFlowBranch = &RigVMFunction_ControlFlowBranch(Condition=RigVMVariableNode() {variable=Is Interacting}.Value, ExecuteContext=@RigVMFunction_Sequence.B)
   n2 = &RigVMFunction_ControlFlowBranch(
    Condition=RigVMDispatchNode(Array=RigVMVariableNode() {variable=Interacting Items}.Value, Element=RigVMVariableNode() {variable=Proxy Control}.Value) {template="DISPATCH_RigVMDispatch_ArrayFind(in Array,in Element,out Index,out Success)"}.Success
    ExecuteContext=@RigVMFunction_ControlFlowBranch.True
   )
   RigVMFunction_Sequence_1 = RigVMAggregateNode(ExecuteContext=@n2.True)
   RigVMVariableNode(ExecuteContext=@RigVMFunction_Sequence_1.C) {variable=For_Each_Count}
   > For_Each = RigVMDispatchNode(Array=RigVMVariableNode() {variable=Driven controls}.Value) {template="DISPATCH_RigVMDispatch_ArrayIterator(in Array,out Element,out Index,out Count,out Ratio)"}
   > RigVMVariableNode(Value=@For_Each.Element) {variable=Current Driven Control}
   RigVMFunction_MathQuaternionToEuler = &RigVMFunction_MathQuaternionToEuler(
    RotationOrder=ZYX
    Value=&RigUnit_GetTransform(Space=LocalSpace, bInitial=False, Item=RigVMVariableNode() {variable=Current Driven Control}.Value).Transform.Rotation
   )
   Add = /Script/RigVM.RigVMFunction_MathVectorAdd(A=@RigVMFunction_MathQuaternionToEuler.Result, B=RigVMVariableNode() {variable=Proxy Rotation Delta}.Value) {is=struct RigVMUnitNode, template="Add::Execute(in A,in B,out Result)"}
   RigVMFunction_MathQuaternionFromEuler = &RigVMFunction_MathQuaternionFromEuler(Euler=@Add.Result)
   > Set Transform = &RigUnit_SetRotation(
    Weight=1
    bInitial=False
    Value=@RigVMFunction_MathQuaternionFromEuler.Result
    Item=RigVMVariableNode() {variable=Current Driven Control}.Value
   )
   n3 = RigVMVariableNode() {variable=Proxy Control}
   RigVMVariableNode(
    Value=&RigVMFunction_MathQuaternionToEuler(RotationOrder=ZYX, Value=&RigUnit_GetTransform(Space=LocalSpace, bInitial=False, Item=@n3.Value).Transform.Rotation).Result
    ExecuteContext=@RigVMFunction_Sequence_1.A
   ) {variable=Proxy Euler rotation}
   RigVMVariableNode(
    Value=/Script/RigVM.RigVMFunction_MathVectorSub(B=RigVMVariableNode() {variable=Origin Rotation}.Value, A=RigVMVariableNode() {variable=Proxy Euler rotation}.Value) {is=struct RigVMUnitNode, template="Subtract::Execute(in A,in B,out Result)"}.Result
    ExecuteContext=@RigVMFunction_Sequence_1.B
   ) {variable=Proxy Rotation Delta}
   RigVMVariableNode(Value=RigVMVariableNode() {variable=Proxy Euler rotation}.Value, ExecuteContext=@RigVMFunction_Sequence_1.D) {variable=Origin Rotation}
   RigVMVariableNode(Value=(X=0,Y=0,Z=0), ExecuteContext=@RigVMFunction_ControlFlowBranch.False) {variable=Origin Rotation}
   > &RigUnit_SetTransform(
    Space=LocalSpace
    Value=(Rotation=(X=0,Y=0,Z=0,W=1),Translation=(X=0,Y=0,Z=0),Scale3D=(X=1,Y=1,Z=1))
    Weight=1
    bInitial=False
    Item=RigVMVariableNode() {variable=Proxy Control}.Value
   )
   /Script/RigVM.RigVMFunction_MathIntAdd(A=0, B=1) {is=struct RigVMUnitNode, template="Add::Execute(in A,in B,out Result)"}
  ]}
 RigVMModel Backwards Solve Graph {nodes [/Script/ControlRig.RigUnit_InverseExecution() {is=struct RigVMUnitNode}]}
 RigVMModel Construction Event {nodes [
   Concat = /Script/RigVM.RigVMFunction_NameConcat(B=" Bfr", A=RigVMVariableNode() {variable=Proxy Name}.Value) {is=struct RigVMUnitNode, template="Concat::Execute(in A,in B,out Result)"}
   n1 = &RigUnit_HierarchyAddNull(
    Space=GlobalSpace
    Name=&RigVMFunction_NameConcat(B=" Null", A=RigVMVariableNode() {variable=Proxy Name}.Value) {template="Concat::Execute(in A,in B,out Result)"}.Result
    Transform=&RigUnit_GetTransform(bInitial=true, Item=RigVMVariableNode() {variable=Proxy Pivot Snap to}.Value).Transform
    Parent=RigVMVariableNode() {variable=Parent}.Value
    ExecutePin=@RigVMFunction_Sequence.B
   )
   > SpawnControl = /Script/ControlRig.RigUnit_HierarchyAddControlTransform(
    InitialValue=(Rotation=(X=0,Y=0,Z=0,W=1),Translation=(X=0,Y=0,Z=0),Scale3D=(X=1,Y=1,Z=1))
    OffsetSpace=LocalSpace
    Settings="(DisplayName=\"None\",bSelectable=True,InitialSpace=LocalSpace,bUsePreferredRotationOrder=False,PreferredRotationOrder=YZX,Limits=(LimitTranslationX=(bMinimum=False,bMaximum=False),LimitTranslationY=(bMinimum=False,bMaximum=False),LimitTranslationZ=(bMinimum ...(truncated)"
    Parent=@n1.Item
    Settings.Shape.Transform=RigVMVariableNode() {variable=Transform}.Value
    OffsetTransform=RigVMVariableNode() {variable=Proxy Offset Transform}.Value
    Name=RigVMVariableNode() {variable=Proxy Name}.Value
    Settings.Shape.Name=RigVMVariableNode() {variable=Shape}.Value
    Settings.Shape.Color=RigVMVariableNode() {variable=Color}.Value
    Settings.Proxy.DrivenControls=RigVMVariableNode() {variable=Driven controls}.Value
   ) {
    is=struct RigVMUnitNode
    template="SpawnControl::Execute(in InitialValue,in Settings,in OffsetTransform,in OffsetSpace,in Parent,in Name,out Item)"
   }
   n2 = &RigUnit_HierarchyAddNull(
    ExecutePin=()
    Space=GlobalSpace
    Transform=(Rotation=(X=0,Y=0,Z=0,W=1),Translation=(X=0,Y=0,Z=0),Scale3D=(X=1,Y=1,Z=1))
    Name=@Concat.Result
    Parent=@SpawnControl.Item
   )
   > RigVMDispatchNode(Name=IsSet, NameSpace=Self, Value=False, Item=@n2.Item) {template="DISPATCH_RigDispatch_SetMetadata(in Item,in Name,in NameSpace,in Value,out Success)"}
   n3 = &RigUnit_PrepareForExecution()
   RigVMFunction_Sequence = &RigVMFunction_Sequence(ExecuteContext=@n3.ExecutePin)
   RigVMVariableNode(
    Value=/Script/ControlRig.RigUnit_ResolveArrayConnector(Connector=(Type=Connector,Name="Driven Controls"), SkipSocket=False) {is=struct RigVMUnitNode}.Result
    ExecuteContext=@RigVMFunction_Sequence.A
   ) {variable=Driven controls}
   > RigVMVariableNode(Value=/Script/ControlRig.RigUnit_ResolveConnector(Connector=(Type=Connector,Name="Parent"), SkipSocket=False) {is=struct RigVMUnitNode}.Result) {variable=Parent}
   > RigVMVariableNode(Value=&RigUnit_ResolveConnector(Connector=(Type=Connector,Name="Snap To"), SkipSocket=False).Result) {variable=Proxy Pivot Snap to}
   SpawnAnimationChannel = /Script/ControlRig.RigUnit_HierarchyAddAnimationChannelBool(
    InitialValue=true
    LimitsEnabled=()
    MaximumValue=True
    MinimumValue=False
    Name=Pivot Vis
    Parent=@SpawnControl.Item
    ExecutePin=@n4.ExecuteContext
   ) {
    is=struct RigVMUnitNode
    template="SpawnAnimationChannel::Execute(in InitialValue,in MinimumValue,in MaximumValue,in LimitsEnabled,in Parent,in Name,out Item)"
   }
   RigVMRerouteNode(Value=@SpawnAnimationChannel.Item)
   RigVMRerouteNode(Value=@SpawnControl.Item)
   RigVMRerouteNode(Value=@n2.Item)
   > RigVMVariableNode(Value=@SpawnAnimationChannel.Item) {variable=Proxy Vis channel}
   n4 = RigVMVariableNode(ExecuteContext=@SpawnControl.ExecutePin, Value=@SpawnControl.Item) {variable=Proxy Control}
  ]}
]
hierarchy [
 Parent {type=Connector, loc=<0 0 0>, rot=<0 0 0 1>}
 Root {type=Socket, loc=<0 0 0>, rot=<0 0 0 1>}
 Driven Controls {type=Connector, loc=<0 0 0>, rot=<0 0 0 1>}
 Snap To {type=Connector, loc=<0 0 0>, rot=<0 0 0 1>}
]
variables [
 Color {type=FLinearColor, default=(R=1,G=0.07,B=0,A=1)}
 Current Driven Control {type=FRigElementKey, default=(Type=None,Name="")}
 Driven controls {type="TArray<FRigElementKey>"}
 For_Each_Count {type=int32, default=0}
 Interacting Items {type="TArray<FRigElementKey>"}
 Is Interacting {type=bool, default=False}
 Origin Rotation {type=FVector, default=(X=0,Y=0,Z=0)}
 Parent {type=FRigElementKey, default=(Type=Bone,Name="")}
 Proxy Control {type=FRigElementKey, default=(Type=Bone,Name="")}
 Proxy Control Settings {
  type=FRigUnit_HierarchyAddControlTransform_Settings
  default="(InitialSpace=LocalSpace,bUsePreferredRotationOrder=False,PreferredRotationOrder=YZX,Limits=(LimitTranslationX=(bMinimum=False,bMaximum=False),LimitTranslationY=(bMinimum=False,bMaximum=False),LimitTranslationZ=(bMinimum=False,bMaximum=False),LimitPitch=(b ...(truncated)"
 }
 Proxy Euler rotation {type=FVector, default=(X=0,Y=0,Z=0)}
 Proxy Name {type=FName, default=None}
 Proxy Offset Transform {type=FTransform, default=(Rotation=(X=0,Y=0,Z=0,W=1),Translation=(X=0,Y=0,Z=0),Scale3D=(X=1,Y=1,Z=1))}
 Proxy Pivot Snap to {type=FRigElementKey, default=(Type=None,Name="")}
 Proxy Rotation Delta {type=FVector, default=(X=0,Y=0,Z=0)}
 Proxy Vis channel {type=FRigElementKey, default=(Type=Bone,Name="")}
 Shape {type=FName, default=None}
 Transform {type=FTransform, default=(Rotation=(X=0,Y=0,Z=0,W=1),Translation=(X=0,Y=0,Z=0),Scale3D=(X=0.1,Y=0.1,Z=0.1))}
]
