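# Modular Control Rig module that spawns a proxy control for moving a set of driven controls together.
# Construction resolves connectors and spawns the control under a null at Snap To, falling back to Parent, plus a Pivot Vis channel.
# During interaction, its per-update local delta offsets each driven control, weighted by a profile curve over array position.
# Otherwise, the proxy re-snaps to its pivot and zeroes its local transform.
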ControlRigBlueprint /EpicControlRig/Modules/Miscellaneous/CRM_Epic_ProxyControl_v02
graphs [
 RigVMFunctionLibrary {links [], nodes []}
 RigVMFunction_Sequence_1_ContainedGraph {nodes [
   n1 = RigVMFunctionEntryNode()
   RigVMFunction_Sequence = &RigVMFunction_Sequence(ExecuteContext=@n1.ExecuteContext)
   n2 = &RigVMFunction_Sequence(ExecuteContext=@RigVMFunction_Sequence.B)
   RigVMFunctionReturnNode(A=@RigVMFunction_Sequence.A, B=@n2.A, C=@n2.B)
  ]}
 RigVMFunction_Sequence_2_ContainedGraph {nodes [
   n1 = RigVMFunctionEntryNode()
   n2 = &RigVMFunction_Sequence(ExecuteContext=@n1.ExecuteContext)
   n3 = &RigVMFunction_Sequence(ExecuteContext=@n2.B)
   n4 = &RigVMFunction_Sequence(ExecuteContext=@n3.B)
   RigVMFunctionReturnNode(A=@n2.A, B=@n3.A, C=@n4.A, D=@n4.B)
  ]}
 RigVMModel {nodes [
   RigUnit_BeginExecution = &RigUnit_BeginExecution()
   RigVMFunction_Sequence = &RigVMFunction_Sequence(ExecuteContext=@RigUnit_BeginExecution.ExecutePin)
   n1 = /Script/ControlRig.RigUnit_IsInteracting() {is=struct RigVMUnitNode}
   RigVMFunction_MathBoolAnd = &RigVMFunction_MathBoolAnd(
    B=RigVMDispatchNode(Array=@n1.Items, Element=RigVMVariableNode() {variable=Proxy Control}.Value) {template="DISPATCH_RigVMDispatch_ArrayFind(in Array,in Element,out Index,out Success)"}.Success
    A=@n1.bIsInteracting
   )
   RigVMFunction_ControlFlowBranch = &RigVMFunction_ControlFlowBranch(Condition=@RigVMFunction_MathBoolAnd.Result, ExecuteContext=@RigVMFunction_Sequence.A)
   RigVMFunction_Sequence_1 = RigVMAggregateNode(ExecuteContext=@RigVMFunction_ControlFlowBranch.True)
   For_Each = RigVMDispatchNode(Array=RigVMVariableNode() {variable=Driven controls}.Value, ExecuteContext=@RigVMFunction_Sequence_1.C) {template="DISPATCH_RigVMDispatch_ArrayIterator(in Array,out Element,out Index,out Count,out Ratio)"}
   RigVMFunction_AnimEvalRichCurve = &RigVMFunction_AnimEvalRichCurve(SourceMinimum=0, TargetMinimum=0, Value=@For_Each.Ratio, Curve=RigVMVariableNode() {variable=Driven Control Profile}.Value)
   Interpolate = /Script/RigVM.RigVMFunction_MathTransformLerp(
    A=(Rotation=(X=0,Y=0,Z=0,W=1),Translation=(X=0,Y=0,Z=0),Scale3D=(X=1,Y=1,Z=1))
    T=@RigVMFunction_AnimEvalRichCurve.Result
    B=RigVMVariableNode() {variable=Proxy Local Transform Delta}.Value
   ) {is=struct RigVMUnitNode, template="Interpolate::Execute(in A,in B,in T,out Result)"}
   > &RigUnit_OffsetTransformForItem(Weight=1, bPropagateToChildren=True, Item=@For_Each.Element, OffsetTransform=@Interpolate.Result)
   RigVMVariableNode(
    Value=&RigUnit_GetTransform(Space=LocalSpace, bInitial=False, Item=RigVMVariableNode() {variable=Proxy Control}.Value).Transform
    ExecuteContext=@RigVMFunction_Sequence_1.A
   ) {variable=Proxy Local Transform}
   RigVMVariableNode(
    Value=&RigVMFunction_MathTransformMakeRelative(
     Global=RigVMVariableNode() {variable=Proxy Local Transform}.Value
     Parent=RigVMVariableNode() {variable=Previous Proxy Local Transform}.Value
    ).Local
    ExecuteContext=@RigVMFunction_Sequence_1.B
   ) {variable=Proxy Local Transform Delta}
   RigVMVariableNode(Value=RigVMVariableNode() {variable=Proxy Local Transform}.Value, ExecuteContext=@RigVMFunction_Sequence_1.D) {variable=Previous Proxy Local Transform}
   n2 = RigVMVariableNode() {variable=Proxy Pivot Snap to}
   n3 = RigVMVariableNode() {variable=Proxy Control}
   SetControlOffset = /Script/ControlRig.RigUnit_SetControlOffset(
    Space=GlobalSpace
    Offset=/Script/RigVM.RigVMFunction_MathTransformMakeAbsolute(
     Parent=&RigUnit_GetTransform(bInitial=False, Item=@n2.Value).Transform
     Local=RigVMVariableNode() {variable=Proxy Offset Transform}.Value
    ) {is=struct RigVMUnitNode, template="Make Absolute::Execute(in Local,in Parent,out Global)"}.Global
    Control=@n3.Value.Name
    ExecutePin=@RigVMFunction_ControlFlowBranch.False
   ) {is=struct RigVMUnitNode, template="SetControlOffset::Execute(in Control,io Offset,in Space)"}
   > &RigUnit_SetTransform(
    Space=LocalSpace
    Value=(Rotation=(X=0,Y=0,Z=0,W=1),Translation=(X=0,Y=0,Z=0),Scale3D=(X=1,Y=1,Z=1))
    Weight=1
    bInitial=False
    Item=@n3.Value
   )
   > RigVMVariableNode(Value=(Rotation=(X=0,Y=0,Z=0,W=1),Translation=(X=0,Y=0,Z=0),Scale3D=(X=1,Y=1,Z=1))) {variable=Previous Proxy Local Transform}
  ]}
 RigVMModel Backwards Solve Graph {nodes [/Script/ControlRig.RigUnit_InverseExecution() {is=struct RigVMUnitNode}]}
 RigVMModel Construction Event {nodes [
   n1 = &RigUnit_PrepareForExecution()
   RigVMFunction_Sequence = RigVMAggregateNode(ExecuteContext=@n1.ExecutePin)
   SpawnAnimationChannel = /Script/ControlRig.RigUnit_HierarchyAddAnimationChannelBool(
    InitialValue=true
    LimitsEnabled=()
    MaximumValue=True
    MinimumValue=False
    Name=Pivot Vis
    Parent=RigVMVariableNode() {variable=Proxy Control}.Value
    ExecutePin=@RigVMFunction_Sequence.C
   ) {
    is=struct RigVMUnitNode
    template="SpawnAnimationChannel::Execute(in InitialValue,in MinimumValue,in MaximumValue,in LimitsEnabled,in Parent,in Name,out Item)"
   }
   > RigVMVariableNode(Value=@SpawnAnimationChannel.Item) {variable=Proxy Vis channel}
   RigVMVariableNode(
    Value=/Script/ControlRig.RigUnit_ResolveArrayConnector(Connector=(Type=Connector,Name="Driven Controls"), SkipSocket=False) {is=struct RigVMUnitNode}.Result
    ExecuteContext=@RigVMFunction_Sequence.A
   ) {variable=Driven controls}
   > RigVMVariableNode(Value=/Script/ControlRig.RigUnit_ResolveConnector(Connector=(Type=Connector,Name="Parent"), SkipSocket=False) {is=struct RigVMUnitNode}.Result) {variable=Parent}
   n2 = &RigUnit_ResolveConnector(Connector=(Type=Connector,Name="Snap To"), SkipSocket=False)
   > RigVMVariableNode(Value=RigVMDispatchNode(True=@n2.Result, Condition=@n2.bIsConnected, False=RigVMVariableNode() {variable=Parent}.Value) {template="DISPATCH_RigVMDispatch_If(in Condition,in True,in False,out Result)"}.Result) {variable=Proxy Pivot Snap to}
   n3 = &RigUnit_HierarchyAddNull(
    Space=GlobalSpace
    Name=/Script/RigVM.RigVMFunction_NameConcat(B=" Null", A=RigVMVariableNode() {variable=Proxy Name}.Value) {is=struct RigVMUnitNode, template="Concat::Execute(in A,in B,out Result)"}.Result
    Transform=&RigUnit_GetTransform(bInitial=true, Item=RigVMVariableNode() {variable=Proxy Pivot Snap to}.Value).Transform
    Parent=RigVMVariableNode() {variable=Parent}.Value
    ExecutePin=@RigVMFunction_Sequence.B
   )
   n4 = RigVMVariableNode() {variable=Proxy Control Shape Settings}
   Get Control Color From Metadata = RigVMFunctionReferenceNode(Color Override=@n4.Value.Color)
   > SpawnControl = /Script/ControlRig.RigUnit_HierarchyAddControlTransform(
    InitialValue=(Rotation=(X=0,Y=0,Z=0,W=1),Translation=(X=0,Y=0,Z=0),Scale3D=(X=1,Y=1,Z=1))
    OffsetSpace=LocalSpace
    Settings="(DisplayName=\"None\",bSelectable=True,InitialSpace=LocalSpace,bUsePreferredRotationOrder=False,PreferredRotationOrder=YZX,Limits=(LimitTranslationX=(bMinimum=False,bMaximum=False),LimitTranslationY=(bMinimum=False,bMaximum=False),LimitTranslationZ=(bMinimum ...(truncated)"
    Settings.Shape.Color=@"Get Control Color From Metadata".Result
    Parent=@n3.Item
    OffsetTransform=RigVMVariableNode() {variable=Proxy Offset Transform}.Value
    Settings.Shape.Name=@n4.Value.Name
    Settings.Shape.Transform=@n4.Value.Transform
    Settings.Shape.bVisible=@n4.Value.bVisible
    Name=RigVMVariableNode() {variable=Proxy Name}.Value
    Settings.Proxy.DrivenControls=RigVMVariableNode() {variable=Driven controls}.Value
   ) {
    is=struct RigVMUnitNode
    template="SpawnControl::Execute(in InitialValue,in Settings,in OffsetTransform,in OffsetSpace,in Parent,in Name,out Item)"
   }
   > RigVMVariableNode(Value=@SpawnControl.Item) {variable=Proxy Control}
  ]}
]
hierarchy [
 Parent {type=Connector, loc=<0 0 0>, rot=<0 0 0 1>}
 Root {type=Socket, loc=<0 0 0>, rot=<0 0 0 1>}
 Driven Controls {type=Connector, loc=<0 0 0>, rot=<0 0 0 1>}
 Snap To {type=Connector, loc=<0 0 0>, rot=<0 0 0 1>}
]
variables [
 Driven Control Profile {
  type=FRuntimeFloatCurve
  default=(EditorCurveData=(Keys=((Value=1),(Time=1,Value=1)),DefaultValue=340282346638528859811704183484516925440,PreInfinityExtrap=RCCE_Constant,PostInfinityExtrap=RCCE_Constant),ExternalCurve=None)
 }
 Driven controls {type="TArray<FRigElementKey>"}
 Parent {type=FRigElementKey, default=(Type=Bone,Name="")}
 Previous Proxy Local Transform {type=FTransform, default=(Rotation=(X=0,Y=0,Z=0,W=1),Translation=(X=0,Y=0,Z=0),Scale3D=(X=1,Y=1,Z=1))}
 Proxy Control {type=FRigElementKey, default=(Type=Bone,Name="")}
 Proxy Control Shape Settings {
  type=FRigUnit_HierarchyAddControl_ShapeSettings
  default=(bVisible=True,Name="DefaultGizmoLibraryNormalized.Default",Color=(R=1,G=0.07,B=0,A=1),Transform=(Rotation=(X=0,Y=0,Z=0,W=1),Translation=(X=0,Y=0,Z=0),Scale3D=(X=0.1,Y=0.1,Z=0.1)))
 }
 Proxy Local Transform {type=FTransform, default=(Rotation=(X=0,Y=0,Z=0,W=1),Translation=(X=0,Y=0,Z=0),Scale3D=(X=1,Y=1,Z=1))}
 Proxy Local Transform Delta {type=FTransform, default=(Rotation=(X=0,Y=0,Z=0,W=1),Translation=(X=0,Y=0,Z=0),Scale3D=(X=1,Y=1,Z=1))}
 Proxy Name {type=FName, default=None}
 Proxy Offset Transform {type=FTransform, default=(Rotation=(X=0,Y=0,Z=0,W=1),Translation=(X=0,Y=0,Z=0),Scale3D=(X=1,Y=1,Z=1))}
 Proxy Pivot Snap to {type=FRigElementKey, default=(Type=None,Name="")}
 Proxy Vis channel {type=FRigElementKey, default=(Type=Bone,Name="")}
]
