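# Control Rig module for a first-person camera rig, active only when a root_fp bone exists.
# Construction spawns Root FP, with Attach FP and Camera Root beneath it and Camera under Camera Root, plus a shared visibility channel.
# Forward solve drives matching bones from controls and applies visibility.
# Backward solve copies bones onto controls, except Root FP, which reads itself, and Camera, whose item names have swapped casing.
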ControlRigBlueprint /EpicControlRig/Modules/Miscellaneous/CRM_FPSCam_v01
graphs [
 ConstructionGraph {nodes [
   RigUnit_PrepareForExecution = &RigUnit_PrepareForExecution()
   RigVMFunction_ControlFlowBranch = &RigVMFunction_ControlFlowBranch(
    Condition=/Script/ControlRig.RigUnit_ItemExists(Item=(Type=Bone,Name="root_fp")) {is=struct RigVMUnitNode}.Exists
    ExecuteContext=@RigUnit_PrepareForExecution.ExecutePin
   )
   n1 = /Script/ControlRig.RigUnit_ResolveConnector(Connector=(Type=Connector,Name="Root"), SkipSocket=False) {is=struct RigVMUnitNode}
   RigVMVariableNode(Value=@n1.Result, ExecuteContext=@RigVMFunction_ControlFlowBranch.True) {variable=Parent Space}
   Get Control Color From Metadata = RigVMFunctionReferenceNode(Color Override=(R=1,G=1,B=1,A=1))
   n2 = RigVMVariableNode() {variable=Control Scale}
   > SpawnControl = /Script/ControlRig.RigUnit_HierarchyAddControlTransform(
    InitialValue=(Rotation=(X=0,Y=0,Z=0,W=1),Translation=(X=0,Y=0,Z=0),Scale3D=(X=1,Y=1,Z=1))
    Name=Root FP
    OffsetSpace=LocalSpace
    OffsetTransform=(Rotation=(X=0,Y=0,Z=0,W=1),Translation=(X=0,Y=0,Z=0),Scale3D=(X=1,Y=1,Z=1))
    Settings="(InitialSpace=LocalSpace,bUsePreferredRotationOrder=False,PreferredRotationOrder=YZX,Limits=(LimitTranslationX=(bMinimum=False,bMaximum=False),LimitTranslationY=(bMinimum=False,bMaximum=False),LimitTranslationZ=(bMinimum=False,bMaximum=False),LimitPitch=(b ...(truncated)"
    Settings.Shape.Color=@"Get Control Color From Metadata".Result
    Parent=@n1.Result
    Settings.Shape.Transform.Scale3D=/Script/RigVM.RigVMFunction_MathVectorScale(Value=(X=1,Y=1,Z=1), Factor=@n2.Value) {is=struct RigVMUnitNode, template="Scale::Execute(in Value,in Factor,out Result)"}.Result
   ) {
    is=struct RigVMUnitNode
    template="SpawnControl::Execute(in InitialValue,in Settings,in OffsetTransform,in OffsetSpace,in Parent,in Name,out Item)"
   }
   > RigVMVariableNode(Value=@SpawnControl.Item) {variable=Root FP Control}
   > RigVMFunction_Sequence = RigVMAggregateNode()
   n3 = &RigVMFunction_MathVectorScale(Value=(X=1,Y=1,Z=1), Factor=&RigVMFunction_MathDoubleMul(B=0.5, A=RigVMVariableNode() {variable=Control Scale}.Value).Result) {template="Scale::Execute(in Value,in Factor,out Result)"}
   n4 = &RigUnit_HierarchyAddControlTransform(
    InitialValue=(Rotation=(X=0,Y=0,Z=0,W=1),Translation=(X=0,Y=0,Z=0),Scale3D=(X=1,Y=1,Z=1))
    Name=Attach FP
    OffsetSpace=LocalSpace
    OffsetTransform=(Rotation=(X=0,Y=0,Z=0,W=1),Translation=(X=0,Y=0,Z=0),Scale3D=(X=1,Y=1,Z=1))
    Settings="(InitialSpace=LocalSpace,bUsePreferredRotationOrder=False,PreferredRotationOrder=YZX,Limits=(LimitTranslationX=(bMinimum=False,bMaximum=False),LimitTranslationY=(bMinimum=False,bMaximum=False),LimitTranslationZ=(bMinimum=False,bMaximum=False),LimitPitch=(b ...(truncated)"
    ExecutePin=@RigVMFunction_Sequence.B
    Settings.Shape.Transform.Scale3D=@n3.Result
    Parent=@SpawnControl.Item
    Settings.Shape.Color=@"Get Control Color From Metadata".Result
   ) {template="SpawnControl::Execute(in InitialValue,in Settings,in OffsetTransform,in OffsetSpace,in Parent,in Name,out Item)"}
   > RigVMVariableNode(Value=@n4.Item) {variable=Attach FP}
   Multiply = &RigVMFunction_MathDoubleMul(B=0.8, A=@n2.Value)
   Scale = &RigVMFunction_MathVectorScale(Value=(X=1,Y=1,Z=1), Factor=@Multiply.Result) {template="Scale::Execute(in Value,in Factor,out Result)"}
   n5 = &RigUnit_HierarchyAddControlTransform(
    InitialValue=(Rotation=(X=0,Y=0,Z=0,W=1),Translation=(X=0,Y=0,Z=0),Scale3D=(X=1,Y=1,Z=1))
    Name=Camera Root
    OffsetSpace=LocalSpace
    OffsetTransform=(Rotation=(X=0,Y=0,Z=0,W=1),Translation=(X=0,Y=0,Z=0),Scale3D=(X=1,Y=1,Z=1))
    Settings="(InitialSpace=LocalSpace,bUsePreferredRotationOrder=False,PreferredRotationOrder=YZX,Limits=(LimitTranslationX=(bMinimum=False,bMaximum=False),LimitTranslationY=(bMinimum=False,bMaximum=False),LimitTranslationZ=(bMinimum=False,bMaximum=False),LimitPitch=(b ...(truncated)"
    ExecutePin=@RigVMFunction_Sequence.C
    Settings.Shape.Transform.Scale3D=@Scale.Result
    Parent=@SpawnControl.Item
    Settings.Shape.Color=@"Get Control Color From Metadata".Result
   ) {template="SpawnControl::Execute(in InitialValue,in Settings,in OffsetTransform,in OffsetSpace,in Parent,in Name,out Item)"}
   > RigVMVariableNode(Value=@n5.Item) {variable=Camera Root}
   > n6 = &RigUnit_HierarchyAddControlTransform(
    InitialValue=(Rotation=(X=0,Y=0,Z=0,W=1),Translation=(X=0,Y=0,Z=0),Scale3D=(X=1,Y=1,Z=1))
    Name=Camera
    OffsetSpace=LocalSpace
    OffsetTransform=(Rotation=(X=0,Y=0,Z=0,W=1),Translation=(X=0,Y=0,Z=0),Scale3D=(X=1,Y=1,Z=1))
    Settings="(InitialSpace=LocalSpace,bUsePreferredRotationOrder=False,PreferredRotationOrder=YZX,Limits=(LimitTranslationX=(bMinimum=False,bMaximum=False),LimitTranslationY=(bMinimum=False,bMaximum=False),LimitTranslationZ=(bMinimum=False,bMaximum=False),LimitPitch=(b ...(truncated)"
    Settings.Shape.Transform.Scale3D=@n3.Result
    Parent=@n5.Item
    Settings.Shape.Color=@"Get Control Color From Metadata".Result
   ) {template="SpawnControl::Execute(in InitialValue,in Settings,in OffsetTransform,in OffsetSpace,in Parent,in Name,out Item)"}
   > RigVMVariableNode(Value=@n6.Item) {variable=Camera}
   SpawnAnimationChannel = /Script/ControlRig.RigUnit_HierarchyAddAnimationChannelBool(
    InitialValue=false
    LimitsEnabled=()
    MaximumValue=True
    MinimumValue=False
    Name=FPS Control Vis
    ExecutePin=@RigVMFunction_Sequence.E
    Parent=RigVMVariableNode() {variable=Root FP Control}.Value
   ) {
    is=struct RigVMUnitNode
    template="SpawnAnimationChannel::Execute(in InitialValue,in MinimumValue,in MaximumValue,in LimitsEnabled,in Parent,in Name,out Item)"
   }
   > RigVMVariableNode(Value=@SpawnAnimationChannel.Item) {variable=Vis Channel}
   Get_Module_Metadata = RigVMDispatchNode(Name=Global Control, NameSpace=Root, Default=RigVMVariableNode() {variable=Parent Space}.Value) {template="DISPATCH_RigDispatch_GetModuleMetadata(in Name,in NameSpace,in Default,out Value,out Found)"}
   > /Script/ControlRig.RigUnit_SetChannelHosts(
    Hosts=((Type=None,Name="None"),(Type=None,Name="None"),(Type=None,Name="None"),(Type=None,Name="None"))
    Hosts.3=@Get_Module_Metadata.Value
    Channel=@SpawnAnimationChannel.Item
    Hosts.0=RigVMVariableNode() {variable=Attach FP}.Value
    Hosts.1=RigVMVariableNode() {variable=Camera Root}.Value
    Hosts.2=RigVMVariableNode() {variable=Camera}.Value
   ) {is=struct RigVMUnitNode}
   Get_Module_Metadata_1 = RigVMDispatchNode(Default=1, Name=Global Control Scale, NameSpace=Root) {template="DISPATCH_RigDispatch_GetModuleMetadata(in Name,in NameSpace,in Default,out Value,out Found)"}
   Set Control Scale = RigVMFunctionReferenceNode(
    Auto Scale Factor=1
    Controls=()
    Global Scale=@Get_Module_Metadata_1.Value
    Scale=RigVMVariableNode() {variable=Control Scale}.Value
    ExecuteContext=@RigVMFunction_Sequence.D
   )
  ]}
 RigVMFunctionLibrary {links [], nodes []}
 RigVMFunction_Sequence_1_ContainedGraph {nodes [
   n1 = RigVMFunctionEntryNode()
   RigVMFunction_Sequence = &RigVMFunction_Sequence(ExecuteContext=@n1.ExecuteContext)
   n2 = &RigVMFunction_Sequence(ExecuteContext=@RigVMFunction_Sequence.B)
   n3 = &RigVMFunction_Sequence(ExecuteContext=@n2.B)
   n4 = &RigVMFunction_Sequence(ExecuteContext=@n3.B)
   RigVMFunctionReturnNode(A=@RigVMFunction_Sequence.A, B=@n2.A, C=@n3.A, D=@n4.A, E=@n4.B)
  ]}
 RigVMModel {nodes [
   RigUnit_BeginExecution = &RigUnit_BeginExecution()
   RigVMFunction_ControlFlowBranch = &RigVMFunction_ControlFlowBranch(Condition=&RigUnit_ItemExists(Item=(Type=Bone,Name="root_fp")).Exists, ExecuteContext=@RigUnit_BeginExecution.ExecutePin)
   n1 = &RigVMFunction_Sequence(ExecuteContext=@RigVMFunction_ControlFlowBranch.True)
   RigUnit_SetTransform = &RigUnit_SetTransform(
    Item=(Type=Bone,Name="root_fp")
    Weight=1
    bInitial=False
    bPropagateToChildren=true
    Value=&RigUnit_GetTransform(Item=(Type=Control,Name="Root FP"), bInitial=False).Transform
    ExecutePin=@n1.A
   )
   RigVMFunction_Sequence = &RigVMFunction_Sequence(ExecuteContext=@RigUnit_SetTransform.ExecutePin)
   &RigUnit_SetTransform(
    Item=(Type=Bone,Name="camera_root")
    Weight=1
    bInitial=False
    Value=&RigUnit_GetTransform(Item=(Type=Control,Name="Camera Root"), bInitial=False).Transform
    ExecutePin=@RigVMFunction_Sequence.A
   )
   > &RigUnit_SetTransform(
    Item=(Type=Bone,Name="camera")
    Weight=1
    bInitial=False
    Value=&RigUnit_GetTransform(Item=(Type=Control,Name="Camera"), bInitial=False).Transform
   )
   &RigUnit_SetTransform(
    Item=(Type=Bone,Name="attach_fp")
    Weight=1
    bInitial=False
    Value=&RigUnit_GetTransform(Item=(Type=Control,Name="Attach FP"), bInitial=False).Transform
    ExecutePin=@RigVMFunction_Sequence.B
   )
   ItemArray = RigVMDispatchNode(Value=((Type=Control,Name="Root FP"),(Type=Control,Name="Attach FP"),(Type=Control,Name="Camera Root"),(Type=Control,Name="Camera"))) {template="DISPATCH_RigVMDispatch_Constant(io Value)"}
   For_Each = RigVMDispatchNode(Array=@ItemArray.Value, ExecuteContext=@n1.B) {template="DISPATCH_RigVMDispatch_ArrayIterator(in Array,out Element,out Index,out Count,out Ratio)"}
   GetAnimationChannelFromItem = /Script/ControlRig.RigUnit_GetBoolAnimationChannelFromItem(bInitial=False, Item=RigVMVariableNode() {variable=Vis Channel}.Value) {is=struct RigVMUnitNode, template="GetAnimationChannelFromItem::Execute(out Value,in Item,in bInitial)"}
   > SetControlVisibility = /Script/ControlRig.RigUnit_SetControlVisibility(Item=@For_Each.Element, bVisible=@GetAnimationChannelFromItem.Value) {is=struct RigVMUnitNode, template="SetControlVisibility::Execute(in Item,in Pattern,in bVisible)"}
  ]}
 RigVMModel Backwards Solve Graph {nodes [
   n1 = /Script/ControlRig.RigUnit_InverseExecution() {is=struct RigVMUnitNode}
   RigVMFunction_ControlFlowBranch = &RigVMFunction_ControlFlowBranch(ExecuteContext=@n1.ExecutePin, Condition=&RigUnit_ItemExists(Item=(Type=Bone,Name="root_fp")).Exists)
   RigUnit_SetTransform = &RigUnit_SetTransform(
    Item=(Type=Control,Name="Root FP")
    Weight=1
    bInitial=False
    bPropagateToChildren=true
    Value=&RigUnit_GetTransform(Item=(Type=Control,Name="Root FP"), bInitial=False).Transform
    ExecutePin=@RigVMFunction_ControlFlowBranch.True
   )
   RigVMFunction_Sequence = &RigVMFunction_Sequence(ExecuteContext=@RigUnit_SetTransform.ExecutePin)
   &RigUnit_SetTransform(
    Item=(Type=Control,Name="Camera Root")
    Weight=1
    bInitial=False
    Value=&RigUnit_GetTransform(Item=(Type=Bone,Name="camera_root"), bInitial=False).Transform
    ExecutePin=@RigVMFunction_Sequence.A
   )
   > &RigUnit_SetTransform(
    Item=(Type=Control,Name="camera")
    Weight=1
    bInitial=False
    Value=&RigUnit_GetTransform(Item=(Type=Bone,Name="Camera"), bInitial=False).Transform
   )
   &RigUnit_SetTransform(
    Item=(Type=Control,Name="Attach FP")
    Weight=1
    bInitial=False
    Value=&RigUnit_GetTransform(Item=(Type=Bone,Name="attach_fp"), bInitial=False).Transform
    ExecutePin=@RigVMFunction_Sequence.B
   )
  ]}
]
hierarchy [
 Root {type=Connector, loc=<0 0 0>, rot=<0 0 0 1>}
 Root {type=Socket, loc=<0 0 0>, rot=<0 0 0 1>}
]
variables [
 Attach FP {type=FRigElementKey, default=(Type=Bone,Name="")}
 Camera {type=FRigElementKey, default=(Type=Bone,Name="")}
 Camera Root {type=FRigElementKey, default=(Type=Bone,Name="")}
 Control Scale {type=double, default=1}
 Parent Space {type=FRigElementKey, default=(Type=None,Name="")}
 Root FP Control {type=FRigElementKey, default=(Type=Bone,Name="")}
 Vis Channel {type=FRigElementKey, default=(Type=Bone,Name="")}
]
