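# Control Rig module that rocks a connected control over pivot points against a Ground plane, like a box tipping on its edges.
# Construction spawns a proxy "Rocker" control and gathers custom, box or circle pivots.
# Before forward solve, dragging the proxy applies its motion around the pivot lowest along the up axis, re-anchoring when that pivot changes.
# Interaction state persists as proxy metadata, refreshed while idle as the proxy snaps to the control.
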ControlRigBlueprint /EpicControlRig/Modules/Miscellaneous/CRM_Rocker_v01
graphs [
 ConstructionGraph {nodes [
   RigUnit_PrepareForExecution = &RigUnit_PrepareForExecution()
   RigVMVariableNode(
    Value=/Script/ControlRig.RigUnit_ResolveConnector(Connector=(Type=Connector,Name="Control"), SkipSocket=False) {is=struct RigVMUnitNode}.Result
    ExecuteContext=@RigUnit_PrepareForExecution.ExecutePin
   ) {variable=Control}
   > RigVMVariableNode(Value=&RigUnit_ResolveConnector(Connector=(Type=Connector,Name="Ground"), SkipSocket=False).Result) {variable=Ground}
   Cast = RigVMDispatchNode(Value=RigVMVariableNode() {variable=Up Axis}.Value) {template="DISPATCH_RigVMDispatch_CastEnumToInt(in Value,out Result)"}
   n1 = RigVMDispatchNode(
    Values=((X=1,Y=0,Z=0),(X=0,Y=1,Z=0),((X=0,Y=0,Z=0),(X=0,Y=0,Z=0),X=0,Y=0,Z=1),((X=0,Y=0,Z=0),(X=0,Y=0,Z=0),X=-1,Y=0,Z=0),((X=0,Y=0,Z=0),(X=0,Y=0,Z=0),X=0,Y=-1,Z=0),((X=0,Y=0,Z=0),(X=0,Y=0,Z=0),X=0,Y=0,Z=-1))
    Index=@Cast.Result
   ) {template="DISPATCH_RigVMDispatch_SelectInt32(in Index,in Values,out Result)"}
   > RigVMVariableNode(Value=@n1.Result) {variable=Up Axis Vector}
   > RigVMFunction_Sequence = RigVMAggregateNode()
   n2 = RigVMVariableNode() {variable=Control}
   n3 = /Script/ControlRig.RigUnit_HierarchyAddControlTransform(
    InitialValue=(Rotation=(X=0,Y=0,Z=0,W=1),Translation=(X=0,Y=0,Z=0),Scale3D=(X=1,Y=1,Z=1))
    Name=Rocker
    OffsetSpace=GlobalSpace
    Settings="(InitialSpace=LocalSpace,bUsePreferredRotationOrder=False,PreferredRotationOrder=YZX,Limits=(LimitTranslationX=(bMinimum=False,bMaximum=False),LimitTranslationY=(bMinimum=False,bMaximum=False),LimitTranslationZ=(bMinimum=False,bMaximum=False),LimitPitch=(b ...(truncated)"
    OffsetTransform=&RigUnit_GetTransform(bInitial=False, Item=@n2.Value).Transform
    Parent=/Script/ControlRig.RigUnit_HierarchyGetParent(bDefaultParent=True, Child=@n2.Value) {is=struct RigVMUnitNode}.Parent
    ExecutePin=@RigVMFunction_Sequence.B
    Settings.Proxy.DrivenControls.0=RigVMVariableNode() {variable=Control}.Value
    Settings.Shape=RigVMVariableNode() {variable=Proxy Control Shape}.Value
   ) {
    is=struct RigVMUnitNode
    template="SpawnControl::Execute(in InitialValue,in Settings,in OffsetTransform,in OffsetSpace,in Parent,in Name,out Item)"
   }
   > RigVMVariableNode(Value=@n3.Item) {variable=Proxy Control}
   n4 = RigVMVariableNode(ExecuteContext=@RigVMFunction_Sequence.C, Value=RigVMVariableNode() {variable=Custom Pivots}.Value) {variable=Local Pivots}
   n5 = &RigVMFunction_Sequence(ExecuteContext=@n4.ExecuteContext)
   RigVMFunction_ControlFlowBranch = &RigVMFunction_ControlFlowBranch(ExecuteContext=@n5.A, Condition=RigVMVariableNode() {variable=Add Box Pivots}.Value)
   Compute Box Pivots = RigVMFunctionReferenceNode(
    ExecuteContext=@RigVMFunction_ControlFlowBranch.True
    Dimensions=RigVMVariableNode() {variable=Box Dimensions}.Value
    Offset=RigVMVariableNode() {variable=Box Offset}.Value
   )
   > RigVMDispatchNode(Other=@"Compute Box Pivots".Result, Array=RigVMVariableNode() {variable=Local Pivots}.Value) {template="DISPATCH_RigVMDispatch_ArrayAppend(io Array,in Other)"}
   n6 = &RigVMFunction_ControlFlowBranch(ExecuteContext=@n5.B, Condition=RigVMVariableNode() {variable=Add Circle Pivots}.Value)
   Compute Circle Pivots = RigVMFunctionReferenceNode(
    Count=8
    ExecuteContext=@n6.True
    Axis=RigVMVariableNode() {variable=Circle Axis}.Value
    Radius=RigVMVariableNode() {variable=Circle Radius}.Value
    Offset=RigVMVariableNode() {variable=Circle Offset}.Value
   )
   > RigVMDispatchNode(Other=@"Compute Circle Pivots".Result, Array=RigVMVariableNode() {variable=Local Pivots}.Value) {template="DISPATCH_RigVMDispatch_ArrayAppend(io Array,in Other)"}
   SpawnControl = &RigUnit_HierarchyAddControlTransform(
    ExecutePin=()
    InitialValue=(Rotation=(X=0,Y=0,Z=0,W=1),Translation=(X=0,Y=0,Z=0),Scale3D=(X=1,Y=1,Z=1))
    Name=NewControl
    OffsetSpace=GlobalSpace
    Parent=(Type=Bone,Name="None")
    Settings="(InitialSpace=LocalSpace,bUsePreferredRotationOrder=False,PreferredRotationOrder=YZX,Limits=(LimitTranslationX=(bMinimum=False,bMaximum=False),LimitTranslationY=(bMinimum=False,bMaximum=False),LimitTranslationZ=(bMinimum=False,bMaximum=False),LimitPitch=(b ...(truncated)"
    OffsetTransform=&RigUnit_GetTransform(Item=(Type=Bone,Name="rockerB_02"), bInitial=False).Transform
   ) {template="SpawnControl::Execute(in InitialValue,in Settings,in OffsetTransform,in OffsetSpace,in Parent,in Name,out Item)"}
   > RigVMVariableNode(Value=@SpawnControl.Item) {variable=Control}
  ]}
 ContainedGraph {nodes [
   n1 = RigVMFunctionEntryNode()
   > For_Each_1 = RigVMDispatchNode(Array=@n1.Transforms) {template="DISPATCH_RigVMDispatch_ArrayIterator(in Array,out Element,out Index,out Count,out Ratio)"}
   n2 = /Script/RigVM.RigVMFunction_MathDistanceToPlane(
    Point=@For_Each_1.Element.Translation
    PlanePoint=RigVMVariableNode() {variable=Plane Point}.Value
    PlaneNormal=RigVMVariableNode() {variable=Plane Normal}.Value
   ) {is=struct RigVMUnitNode}
   RigVMFunction_MathBoolOr = &RigVMFunction_MathBoolOr(
    A=RigVMDispatchNode(B=0, A=@For_Each_1.Index) {template="DISPATCH_RigVMDispatch_CoreEquals(in A,in B,out Result)"}.Result
    B=/Script/RigVM.RigVMFunction_MathDoubleLess(A=@n2.SignedDistance, B=RigVMVariableNode() {variable=Distance}.Value) {is=struct RigVMUnitNode, template="Less::Execute(in A,in B,out Result)"}.Result
   )
   n3 = &RigVMFunction_ControlFlowBranch(ExecuteContext=@For_Each_1.ExecuteContext, Condition=@RigVMFunction_MathBoolOr.Result)
   RigVMVariableNode(ExecuteContext=@n3.True, Value=@n2.SignedDistance) {variable=Distance}
   > RigVMVariableNode(Value=@For_Each_1.Element) {variable=Transform}
   > RigVMVariableNode(Value=@For_Each_1.Index) {variable=Local Index}
   RigVMFunctionReturnNode(
    ExecuteContext=@For_Each_1.Completed
    Value=RigVMVariableNode() {variable=Transform}.Value
    Index=RigVMVariableNode() {variable=Local Index}.Value
   )
  ]}
 Multiply_1_ContainedGraph {nodes [
   n1 = RigVMFunctionEntryNode()
   Multiply = &RigVMFunction_MathDoubleMul(A=@n1.A, B=@n1.B)
   RigVMFunctionReturnNode(Result=&RigVMFunction_MathDoubleMul(B=@n1.C, A=@Multiply.Result).Result)
  ]}
 New Function_ContainedGraph {nodes [
   RigVMFunctionEntryNode()
   > n1 = RigVMDispatchNode(Array=RigVMVariableNode() {variable=Transforms}.Value) {template="DISPATCH_RigVMDispatch_ArrayReset(io Array)"}
   RigVMFunction_Sequence = &RigVMFunction_Sequence(ExecuteContext=@n1.ExecuteContext)
   n2 = RigVMVariableNode() {variable=Dimensions}
   Negate = /Script/RigVM.RigVMFunction_MathVectorNegate(Value=@n2.Value) {is=struct RigVMUnitNode, template="Negate::Execute(in Value,out Result)"}
   Make_Array = RigVMDispatchNode(
    Values=((X=0,Y=0,Z=0),((X=0,Y=0,Z=0),X=0,Y=0,Z=0),((X=0,Y=0,Z=0),X=0,Y=0,Z=0),((X=0,Y=0,Z=0),X=0,Y=0,Z=0),((X=0,Y=0,Z=0),X=0,Y=0,Z=0),((X=0,Y=0,Z=0),X=0,Y=0,Z=0),((X=0,Y=0,Z=0),X=0,Y=0,Z=0),((X=0,Y=0,Z=0),X=0,Y=0,Z=0))
    Values.1=@Negate.Result
    Values.2.X=@Negate.Result.X
    Values.5.X=@Negate.Result.X
    Values.7.X=@Negate.Result.X
    Values.3.Y=@Negate.Result.Y
    Values.5.Y=@Negate.Result.Y
    Values.6.Y=@Negate.Result.Y
    Values.4.Z=@Negate.Result.Z
    Values.6.Z=@Negate.Result.Z
    Values.7.Z=@Negate.Result.Z
    Values.0=@n2.Value
    Values.3.X=@n2.Value.X
    Values.4.X=@n2.Value.X
    Values.6.X=@n2.Value.X
    Values.2.Y=@n2.Value.Y
    Values.4.Y=@n2.Value.Y
    Values.7.Y=@n2.Value.Y
    Values.2.Z=@n2.Value.Z
    Values.3.Z=@n2.Value.Z
    Values.5.Z=@n2.Value.Z
   ) {template="DISPATCH_RigVMDispatch_ArrayMake(in Values,out Array)"}
   For_Each = RigVMDispatchNode(Array=@Make_Array.Array, ExecuteContext=@RigVMFunction_Sequence.A) {template="DISPATCH_RigVMDispatch_ArrayIterator(in Array,out Element,out Index,out Count,out Ratio)"}
   Add = /Script/RigVM.RigVMFunction_MathVectorAdd(A=@For_Each.Element, B=RigVMVariableNode() {variable=Offset}.Value) {is=struct RigVMUnitNode, template="Add::Execute(in A,in B,out Result)"}
   RigVMFunction_MathTransformMake = /Script/RigVM.RigVMFunction_MathTransformMake(Rotation=(X=0,Y=0,Z=0,W=1), Scale=(X=1,Y=1,Z=1), Translation=@Add.Result) {is=struct RigVMUnitNode}
   > RigVMDispatchNode(Element=@RigVMFunction_MathTransformMake.Result, Array=RigVMVariableNode() {variable=Transforms}.Value) {template="DISPATCH_RigVMDispatch_ArrayAdd(io Array,in Element,out Index)"}
   RigVMFunctionReturnNode(ExecuteContext=@RigVMFunction_Sequence.B, Result=RigVMVariableNode() {variable=Transforms}.Value)
  ]}
 New Function_ContainedGraph {nodes [
   RigVMFunctionEntryNode()
   > RigVMDispatchNode(Array=RigVMVariableNode() {variable=Transforms}.Value) {template="DISPATCH_RigVMDispatch_ArrayReset(io Array)"}
   > RigVMFunction_Sequence = RigVMAggregateNode()
   n1 = RigVMVariableNode() {variable=Count}
   RigVMFunction_ForLoopCount = /Script/RigVM.RigVMFunction_ForLoopCount(ExecutePin=@RigVMFunction_Sequence.A, Count=@n1.Value) {is=struct RigVMUnitNode}
   Cast = RigVMDispatchNode(Value=RigVMVariableNode() {variable=Axis}.Value) {template="DISPATCH_RigVMDispatch_CastEnumToInt(in Value,out Result)"}
   Make_Array = RigVMDispatchNode(Values=((X=1,Y=0,Z=0),((X=0,Y=0,Z=0),X=0,Y=1,Z=0),((X=0,Y=0,Z=0),X=0,Y=0,Z=1),((X=0,Y=0,Z=0),X=-1,Y=0,Z=0),((X=0,Y=0,Z=0),X=0,Y=-1,Z=0),((X=0,Y=0,Z=0),X=0,Y=0,Z=-1))) {template="DISPATCH_RigVMDispatch_ArrayMake(in Values,out Array)"}
   RigVMFunction_MathIntToDouble = /Script/RigVM.RigVMFunction_MathIntToFloat(Value=@n1.Value) {is=struct RigVMUnitNode, template="Cast::Execute(in Value,out Result)"}
   Divide = &RigVMFunction_MathDoubleDiv(A=1, B=@RigVMFunction_MathIntToDouble.Result)
   TwoPi = /Script/RigVM.RigVMFunction_MathDoubleConstTwoPi() {is=struct RigVMUnitNode, template="TwoPi::Execute(out Value)"}
   Multiply = RigVMAggregateNode(A=@RigVMFunction_ForLoopCount.Ratio, B=&RigVMFunction_MathDoubleSub(A=1, B=@Divide.Result).Result, C=@TwoPi.Value)
   RigVMFunction_MathQuaternionFromAxisAndAngle = /Script/RigVM.RigVMFunction_MathQuaternionFromAxisAndAngle(
    Axis=RigVMDispatchNode(Index=@Cast.Result, Array=@Make_Array.Array) {template="DISPATCH_RigVMDispatch_ArrayGetAtIndex(in Array,in Index,out Element)"}.Element
    Angle=@Multiply.Result
   ) {is=struct RigVMUnitNode, template="FromAxisAndAngle::Execute(in Axis,in Angle,out Result)"}
   Subtract = /Script/RigVM.RigVMFunction_MathIntSub(B=-1, A=@Cast.Result) {is=struct RigVMUnitNode, template="Subtract::Execute(in A,in B,out Result)"}
   RigVMFunction_MathVectorSetLength = /Script/RigVM.RigVMFunction_MathVectorSetLength(
    Value=RigVMDispatchNode(Array=@Make_Array.Array, Index=@Subtract.Result) {template="DISPATCH_RigVMDispatch_ArrayGetAtIndex(in Array,in Index,out Element)"}.Element
    Length=RigVMVariableNode() {variable=Radius}.Value
   ) {is=struct RigVMUnitNode, template="SetLength::Execute(in Value,in Length,out Result)"}
   RigVMFunction_MathTransformMake = &RigVMFunction_MathTransformMake(
    Rotation=(X=0,Y=0,Z=0,W=1)
    Scale=(X=1,Y=1,Z=1)
    Translation=/Script/RigVM.RigVMFunction_MathQuaternionRotateVector(Transform=@RigVMFunction_MathQuaternionFromAxisAndAngle.Result, Vector=@RigVMFunction_MathVectorSetLength.Result) {is=struct RigVMUnitNode, template="Rotate Vector::Execute(in Transform,in Vector,out Result)"}.Result
   )
   > Add = RigVMDispatchNode(
    Element=(Rotation=(X=0,Y=0,Z=0,W=1),Translation=(X=0,Y=0,Z=0),Scale3D=(X=1,Y=1,Z=1))
    Element.Translation=&RigVMFunction_MathVectorAdd(A=@RigVMFunction_MathTransformMake.Result.Translation, B=RigVMVariableNode() {variable=Offset}.Value) {template="Add::Execute(in A,in B,out Result)"}.Result
    Element.Rotation=@RigVMFunction_MathTransformMake.Result.Rotation
    Element.Scale3D=@RigVMFunction_MathTransformMake.Result.Scale3D
    Array=RigVMVariableNode() {variable=Transforms}.Value
   ) {template="DISPATCH_RigVMDispatch_ArrayAdd(io Array,in Element,out Index)"}
   RigVMFunctionReturnNode(ExecuteContext=@RigVMFunction_Sequence.C, Result=RigVMVariableNode() {variable=Transforms}.Value)
  ]}
 New Function_ContainedGraph {nodes [
   n1 = RigVMFunctionEntryNode()
   > RigVMDispatchNode(Array=RigVMVariableNode() {variable=Transforms}.Value) {template="DISPATCH_RigVMDispatch_ArrayReset(io Array)"}
   > For_Each = RigVMDispatchNode(Array=@n1.Local) {template="DISPATCH_RigVMDispatch_ArrayIterator(in Array,out Element,out Index,out Count,out Ratio)"}
   > Add = RigVMDispatchNode(
    Element=/Script/RigVM.RigVMFunction_MathTransformMakeAbsolute(Parent=@n1.Parent, Local=@For_Each.Element) {is=struct RigVMUnitNode, template="Make Absolute::Execute(in Local,in Parent,out Global)"}.Global
    Array=RigVMVariableNode() {variable=Transforms}.Value
   ) {template="DISPATCH_RigVMDispatch_ArrayAdd(io Array,in Element,out Index)"}
   RigVMFunctionReturnNode(ExecuteContext=@For_Each.Completed, Result=RigVMVariableNode() {variable=Transforms}.Value)
  ]}
 New Function_ContainedGraph {nodes [
   RigVMFunctionEntryNode()
   > RigVMFunction_Sequence = RigVMAggregateNode()
   Calculate Body Mover Pivot = RigVMFunctionReferenceNode(
    ExecuteContext=@RigVMFunction_Sequence.A
    Control=RigVMVariableNode() {variable=Control}.Value
    Ground Transform=RigVMVariableNode() {variable=Ground Transform}.Value
    Up Axis=RigVMVariableNode() {variable=Up Axis}.Value
    Pivots=RigVMVariableNode() {variable=Pivots}.Value
   )
   > RigVMVariableNode(Value=@"Calculate Body Mover Pivot".Pivot) {variable=Pivot Transform}
   > RigVMVariableNode(Value=@"Calculate Body Mover Pivot".Index) {variable=Pivot Index}
   DISPATCH_RigDispatch_GetMetadata = RigVMDispatchNode(Default=0, Name=Rocker Current Transform Index, NameSpace=Self, Item=RigVMVariableNode() {variable=Proxy}.Value) {template="DISPATCH_RigDispatch_GetMetadata(in Item,in Name,in NameSpace,in Default,out Value,out Found)"}
   DISPATCH_RigDispatch_CoreNotEquals = RigVMDispatchNode(B=@DISPATCH_RigDispatch_GetMetadata.Value, A=RigVMVariableNode() {variable=Pivot Index}.Value) {template="DISPATCH_RigVMDispatch_CoreNotEquals(in A,in B,out Result)"}
   n1 = &RigVMFunction_ControlFlowBranch(Condition=@DISPATCH_RigDispatch_CoreNotEquals.Result, ExecuteContext=@RigVMFunction_Sequence.B)
   n2 = &RigVMFunction_Sequence(ExecuteContext=@n1.True)
   n3 = RigVMVariableNode() {variable=Control}
   n4 = &RigUnit_GetTransform(bInitial=false, Item=@n3.Value)
   n5 = RigVMVariableNode() {variable=Proxy}
   n6 = RigVMVariableNode() {variable=Control}
   n7 = &RigUnit_HierarchyGetParent(bDefaultParent=True, Child=@n6.Value)
   n8 = &RigUnit_GetTransform(Space=LocalSpace, bInitial=true, Item=@n6.Value)
   n9 = RigVMVariableNode() {variable=Ground Transform}
   n10 = /Script/ControlRig.RigUnit_AimBoneMath(
    DebugSettings=(Scale=10,WorldOffset=(Rotation=(X=0,Y=0,Z=0,W=1),Translation=(X=0,Y=0,Z=0),Scale3D=(X=1,Y=1,Z=1)))
    Primary=(Weight=1,Axis=(X=0,Y=0,Z=1),Target=(X=0,Y=0,Z=0),Kind=Direction,Space=(Type=Bone,Name="None"))
    Secondary=(Weight=1,Axis=(X=1,Y=0,Z=0),Target=(X=1,Y=0,Z=0),Kind=Location,Space=(Type=Bone,Name="None"))
    Weight=1
    InputTransform=&RigVMFunction_MathTransformMakeAbsolute(
     Local=RigVMDispatchNode(
      Default=(Rotation=(X=0,Y=0,Z=0,W=1),Translation=(X=0,Y=0,Z=0),Scale3D=(X=1,Y=1,Z=1))
      Name=Rocker Current Pivot
      NameSpace=Self
      Item=@n5.Value
     ) {template="DISPATCH_RigDispatch_GetMetadata(in Item,in Name,in NameSpace,in Default,out Value,out Found)"}.Value
     Parent=&RigUnit_GetTransform(bInitial=False, Item=@n7.Parent).Transform
    ) {template="Make Absolute::Execute(in Local,in Parent,out Global)"}.Global
    Primary.Target=&RigVMFunction_MathQuaternionRotateVector(
     Transform=/Script/RigVM.RigVMFunction_MathQuaternionMakeAbsolute(Local=@n8.Transform.Rotation, Parent=@n9.Value.Rotation) {is=struct RigVMUnitNode, template="Make Absolute::Execute(in Local,in Parent,out Global)"}.Global
     Vector=RigVMVariableNode() {variable=Up Axis}.Value
    ) {template="Rotate Vector::Execute(in Transform,in Vector,out Result)"}.Result
    Secondary.Target=RigVMVariableNode() {variable=Pivot Transform}.Value.Translation
   ) {is=struct RigVMUnitNode}
   n11 = RigVMVariableNode() {variable=Control}
   n12 = &RigVMFunction_MathTransformMakeAbsolute(
    Local=RigVMDispatchNode(
     Default=(Rotation=(X=0,Y=0,Z=0,W=1),Translation=(X=0,Y=0,Z=0),Scale3D=(X=1,Y=1,Z=1))
     Name=Rocker Current Pivot
     NameSpace=Self
     Item=RigVMVariableNode() {variable=Proxy}.Value
    ) {template="DISPATCH_RigDispatch_GetMetadata(in Item,in Name,in NameSpace,in Default,out Value,out Found)"}.Value
    Parent=&RigUnit_GetTransform(bInitial=False, Item=&RigUnit_HierarchyGetParent(bDefaultParent=True, Child=@n11.Value).Parent).Transform
   ) {template="Make Absolute::Execute(in Local,in Parent,out Global)"}
   n13 = RigVMVariableNode() {variable=Pivot Transform}
   n14 = /Script/RigVM.RigVMFunction_MathTransformTransformVector(
    Location=(X=0,Y=0,Z=0)
    Transform=@n10.Result
    Location.X=/Script/RigVM.RigVMFunction_MathVectorLength(Value=/Script/RigVM.RigVMFunction_MathVectorSub(A=@n12.Global.Translation, B=@n13.Value.Translation) {is=struct RigVMUnitNode, template="Subtract::Execute(in A,in B,out Result)"}.Result) {is=struct RigVMUnitNode, template="Length::Execute(in Value,out Result)"}.Result
   ) {is=struct RigVMUnitNode, template="Multiply::Execute(in Transform,in Location,out Result)"}
   n15 = RigVMVariableNode() {variable=Pivot Transform}
   &RigUnit_SetTransform(
    Value=(Rotation=(X=0,Y=0,Z=0,W=1),Translation=(X=0,Y=0,Z=0),Scale3D=(X=1,Y=1,Z=1))
    Weight=1
    bInitial=False
    Value.Rotation=@n4.Transform.Rotation
    Value.Scale3D=@n4.Transform.Scale3D
    Value.Translation=&RigVMFunction_MathTransformMul(
     B=(Rotation=(X=0,Y=0,Z=0,W=1),Translation=(X=0,Y=0,Z=0),Scale3D=(X=1,Y=1,Z=1))
     A=@n4.Transform
     B.Translation=&RigVMFunction_MathVectorSub(A=@n14.Result, B=@n15.Value.Translation) {template="Subtract::Execute(in A,in B,out Result)"}.Result
    ) {template="Multiply::Execute(in A,in B,out Result)"}.Result.Translation
    ExecutePin=@n2.A
    Item=@n3.Value
   )
   RigVMDispatchNode(
    Name=Rocker Current Transform Index
    NameSpace=Self
    ExecuteContext=@n2.B
    Item=RigVMVariableNode() {variable=Proxy}.Value
    Value=RigVMVariableNode() {variable=Pivot Index}.Value
   ) {template="DISPATCH_RigDispatch_SetMetadata(in Item,in Name,in NameSpace,in Value,out Success)"}
   n16 = &RigUnit_HierarchyGetParent(bDefaultParent=True, Child=RigVMVariableNode() {variable=Control}.Value)
   > DISPATCH_RigDispatch_SetMetadata = RigVMDispatchNode(
    Name=Rocker Current Pivot
    NameSpace=Self
    Value=(Rotation=(X=0,Y=0,Z=0,W=1),Translation=(X=0,Y=0,Z=0),Scale3D=(X=1,Y=1,Z=1))
    Value.Translation=&RigVMFunction_MathTransformMakeRelative(
     Global=(Rotation=(X=0,Y=0,Z=0,W=1),Translation=(X=0,Y=0,Z=0),Scale3D=(X=1,Y=1,Z=1))
     Parent=&RigUnit_GetTransform(bInitial=False, Item=@n16.Parent).Transform
     Global.Translation=@n14.Result
    ).Local.Translation
    Item=RigVMVariableNode() {variable=Proxy}.Value
   ) {template="DISPATCH_RigDispatch_SetMetadata(in Item,in Name,in NameSpace,in Value,out Success)"}
   n17 = RigVMVariableNode() {variable=Control}
   > RigVMDispatchNode(
    Name=Rocker Control Transform
    NameSpace=Self
    Value=&RigUnit_GetRelativeTransformForItem(
     bChildInitial=False
     bParentInitial=False
     Parent=&RigUnit_HierarchyGetParent(bDefaultParent=True, Child=@n17.Value).Parent
     Child=@n17.Value
    ).RelativeTransform
    Item=RigVMVariableNode() {variable=Proxy}.Value
   ) {template="DISPATCH_RigDispatch_SetMetadata(in Item,in Name,in NameSpace,in Value,out Success)"}
   n18 = RigVMVariableNode() {variable=Proxy}
   > RigVMDispatchNode(
    Name=Rocker Proxy Transform
    NameSpace=Self
    Value=&RigUnit_GetRelativeTransformForItem(
     bChildInitial=False
     bParentInitial=False
     Parent=&RigUnit_HierarchyGetParent(bDefaultParent=True, Child=@n18.Value).Parent
     Child=@n18.Value
    ).RelativeTransform
    Item=RigVMVariableNode() {variable=Proxy}.Value
   ) {template="DISPATCH_RigDispatch_SetMetadata(in Item,in Name,in NameSpace,in Value,out Success)"}
   RigVMFunctionReturnNode(ExecuteContext=@RigVMFunction_Sequence.D)
   n19 = RigVMVariableNode() {variable=Control}
   n20 = &RigVMFunction_MathTransformMakeAbsolute(
    Local=RigVMDispatchNode(
     Default=(Rotation=(X=0,Y=0,Z=0,W=1),Translation=(X=0,Y=0,Z=0),Scale3D=(X=1,Y=1,Z=1))
     Name=Rocker Control Transform
     NameSpace=Self
     Item=RigVMVariableNode() {variable=Proxy}.Value
    ) {template="DISPATCH_RigDispatch_GetMetadata(in Item,in Name,in NameSpace,in Default,out Value,out Found)"}.Value
    Parent=&RigUnit_GetTransform(bInitial=False, Item=&RigUnit_HierarchyGetParent(bDefaultParent=True, Child=@n19.Value).Parent).Transform
   ) {template="Make Absolute::Execute(in Local,in Parent,out Global)"}
   n21 = RigVMVariableNode() {variable=Proxy}
   n22 = &RigUnit_HierarchyGetParent(bDefaultParent=True, Child=RigVMVariableNode() {variable=Control}.Value)
   n23 = RigVMVariableNode() {variable=Control}
   n24 = &RigUnit_HierarchyGetParent(bDefaultParent=True, Child=@n23.Value)
   Multiply = &RigVMFunction_MathTransformMul(
    A=(Rotation=(X=0,Y=0,Z=0,W=1),Translation=(X=0,Y=0,Z=0),Scale3D=(X=1,Y=1,Z=1))
    B=&RigVMFunction_MathTransformMakeAbsolute(
     Local=RigVMDispatchNode(
      Default=(Rotation=(X=0,Y=0,Z=0,W=1),Translation=(X=0,Y=0,Z=0),Scale3D=(X=1,Y=1,Z=1))
      Name=Rocker Current Pivot
      NameSpace=Self
      Item=@n21.Value
     ) {template="DISPATCH_RigDispatch_GetMetadata(in Item,in Name,in NameSpace,in Default,out Value,out Found)"}.Value
     Parent=&RigUnit_GetTransform(bInitial=False, Item=@n22.Parent).Transform
    ) {template="Make Absolute::Execute(in Local,in Parent,out Global)"}.Global
    A.Rotation=&RigVMFunction_MathTransformMakeRelative(
     Parent=&RigUnit_GetTransform(bInitial=true, Item=@n24.Parent).Transform
     Global=&RigUnit_GetTransform(bInitial=true, Item=@n23.Value).Transform
    ).Local.Rotation
   ) {template="Multiply::Execute(in A,in B,out Result)"}
   n25 = RigVMVariableNode() {variable=Proxy}
   n26 = &RigUnit_HierarchyGetParent(bDefaultParent=True, Child=@n25.Value)
   n27 = &RigVMFunction_MathTransformMakeRelative(
    Parent=&RigVMFunction_MathTransformMakeAbsolute(
     Local=RigVMDispatchNode(
      Default=(Rotation=(X=0,Y=0,Z=0,W=1),Translation=(X=0,Y=0,Z=0),Scale3D=(X=1,Y=1,Z=1))
      Name=Rocker Proxy Transform
      NameSpace=Self
      Item=@n25.Value
     ) {template="DISPATCH_RigDispatch_GetMetadata(in Item,in Name,in NameSpace,in Default,out Value,out Found)"}.Value
     Parent=&RigUnit_GetTransform(bInitial=False, Item=@n26.Parent).Transform
    ) {template="Make Absolute::Execute(in Local,in Parent,out Global)"}.Global
    Global=&RigUnit_GetTransform(Item=@n25.Value).Transform
   )
   &RigUnit_SetTransform(
    Weight=1
    bInitial=False
    Value=&RigVMFunction_MathTransformMul(
     A=&RigVMFunction_MathTransformMakeRelative(Global=@n20.Global, Parent=@Multiply.Result).Local
     B=&RigVMFunction_MathTransformMul(A=@n27.Local, B=@Multiply.Result) {template="Multiply::Execute(in A,in B,out Result)"}.Result
    ) {template="Multiply::Execute(in A,in B,out Result)"}.Result
    Item=RigVMVariableNode() {variable=Control}.Value
    ExecutePin=@RigVMFunction_Sequence.C
   )
  ]}
 New Function_ContainedGraph {nodes [
   RigVMFunctionEntryNode()
   > RigVMFunction_Sequence = RigVMAggregateNode()
   n1 = RigVMVariableNode() {variable=Control}
   RigVMDispatchNode(
    Name=Rocker Control Transform
    NameSpace=Self
    Value=&RigUnit_GetRelativeTransformForItem(
     bChildInitial=False
     bParentInitial=False
     Parent=&RigUnit_HierarchyGetParent(bDefaultParent=True, Child=@n1.Value).Parent
     Child=@n1.Value
    ).RelativeTransform
    ExecuteContext=@RigVMFunction_Sequence.A
    Item=RigVMVariableNode() {variable=Proxy}.Value
   ) {template="DISPATCH_RigDispatch_SetMetadata(in Item,in Name,in NameSpace,in Value,out Success)"}
   Negate = &RigVMFunction_MathVectorNegate(Value=RigVMVariableNode() {variable=Up Axis}.Value) {template="Negate::Execute(in Value,out Result)"}
   n2 = RigVMVariableNode() {variable=Ground Transform}
   n3 = /Script/RigVM.RigVMFunction_MathTransformRotateVector(Vector=@Negate.Result, Transform=@n2.Value) {is=struct RigVMUnitNode, template="Rotate Vector::Execute(in Transform,in Vector,out Result)"}
   n4 = &RigUnit_GetTransform(bInitial=false, Item=RigVMVariableNode() {variable=Control}.Value)
   n5 = RigVMVariableNode() {variable=Up Axis}
   Make = RigVMDispatchNode(
    Elements=(Rotation=(X=0,Y=0,Z=0,W=1),Translation=(X=0,Y=0,Z=0),Scale3D=(X=1,Y=1,Z=1))
    Elements.Translation=/Script/RigVM.RigVMFunction_MathIntersectPlane(PlanePoint=(X=0,Y=0,Z=0), Direction=@n3.Result, PlaneNormal=@n3.Result, Start=@n4.Transform.Translation) {is=struct RigVMUnitNode}.Result
    Elements.Rotation=&RigVMFunction_MathTransformMul(
     A=(Rotation=(X=0,Y=0,Z=0,W=1),Translation=(X=0,Y=0,Z=0),Scale3D=(X=1,Y=1,Z=1))
     A.Rotation=&RigVMFunction_MathQuaternionSwingTwist(Input=@n4.Transform.Rotation, TwistAxis=@n5.Value).Twist
     B=@n2.Value
    ) {template="Multiply::Execute(in A,in B,out Result)"}.Result.Rotation
   ) {template="DISPATCH_RigVMDispatch_MakeStruct(in Elements,out Struct)"}
   RigVMDispatchNode(
    Name=Rocker Ground Transform
    NameSpace=Self
    Value=@Make.Struct
    ExecuteContext=@RigVMFunction_Sequence.B
    Item=RigVMVariableNode() {variable=Proxy}.Value
   ) {template="DISPATCH_RigDispatch_SetMetadata(in Item,in Name,in NameSpace,in Value,out Success)"}
   Calculate Body Mover Pivot = RigVMFunctionReferenceNode(
    ExecuteContext=@RigVMFunction_Sequence.C
    Up Axis=RigVMVariableNode() {variable=Up Axis}.Value
    Ground Transform=RigVMVariableNode() {variable=Ground Transform}.Value
    Control=RigVMVariableNode() {variable=Control}.Value
    Pivots=RigVMVariableNode() {variable=World Pivots}.Value
   )
   n6 = RigVMVariableNode() {variable=Proxy}
   > RigVMDispatchNode(Name=Rocker Current Transform Index, NameSpace=Self, Value=@"Calculate Body Mover Pivot".Index, Item=@n6.Value) {template="DISPATCH_RigDispatch_SetMetadata(in Item,in Name,in NameSpace,in Value,out Success)"}
   n7 = &RigUnit_HierarchyGetParent(bDefaultParent=True, Child=RigVMVariableNode() {variable=Control}.Value)
   > RigVMDispatchNode(
    Name=Rocker Current Pivot
    NameSpace=Self
    Value=(Rotation=(X=0,Y=0,Z=0,W=1),Translation=(X=0,Y=0,Z=0),Scale3D=(X=1,Y=1,Z=1))
    Value.Translation=&RigVMFunction_MathTransformMakeRelative(
     Global=(Rotation=(X=0,Y=0,Z=0,W=1),Translation=(X=0,Y=0,Z=0),Scale3D=(X=1,Y=1,Z=1))
     Global.Translation=@"Calculate Body Mover Pivot".Pivot.Translation
     Parent=&RigUnit_GetTransform(bInitial=False, Item=@n7.Parent).Transform
    ).Local.Translation
    Item=@n6.Value
   ) {template="DISPATCH_RigDispatch_SetMetadata(in Item,in Name,in NameSpace,in Value,out Success)"}
   n8 = RigVMVariableNode() {variable=Proxy}
   RigVMDispatchNode(
    Name=Rocker Proxy Transform
    NameSpace=Self
    Value=&RigUnit_GetRelativeTransformForItem(
     bChildInitial=False
     bParentInitial=False
     Parent=&RigUnit_HierarchyGetParent(bDefaultParent=True, Child=@n8.Value).Parent
     Child=@n8.Value
    ).RelativeTransform
    ExecuteContext=@RigVMFunction_Sequence.D
    Item=@n8.Value
   ) {template="DISPATCH_RigDispatch_SetMetadata(in Item,in Name,in NameSpace,in Value,out Success)"}
   RigVMFunctionReturnNode(ExecuteContext=@RigVMFunction_Sequence.E)
  ]}
 New Function_ContainedGraph {nodes [
   RigVMFunctionEntryNode()
   n1 = RigVMVariableNode() {variable=Control}
   n2 = &RigUnit_GetTransform(Space=LocalSpace, bInitial=true, Item=@n1.Value)
   n3 = RigVMVariableNode() {variable=Up Axis}
   n4 = RigVMVariableNode() {variable=Ground Transform}
   > Get Transform Closest To Plane = RigVMFunctionReferenceNode(
    Plane Normal=&RigVMFunction_MathQuaternionRotateVector(
     Vector=&RigVMFunction_MathQuaternionRotateVector(Transform=@n2.Transform.Rotation, Vector=@n3.Value) {template="Rotate Vector::Execute(in Transform,in Vector,out Result)"}.Result
     Transform=@n4.Value.Rotation
    ) {template="Rotate Vector::Execute(in Transform,in Vector,out Result)"}.Result
    Plane Point=@n4.Value.Translation
    Transforms=RigVMVariableNode() {variable=Pivots}.Value
   )
   > RigVMFunctionReturnNode(Index=@"Get Transform Closest To Plane".Index, Pivot=@"Get Transform Closest To Plane".Value)
   &RigVMFunction_MathQuaternionMakeAbsolute(
    Parent=&RigUnit_GetTransform(bInitial=false, Item=&RigUnit_HierarchyGetParent(bDefaultParent=True, Child=@n1.Value).Parent).Transform.Rotation
    Local=@n2.Transform.Rotation
   ) {template="Make Absolute::Execute(in Local,in Parent,out Global)"}
  ]}
 RigVMFunctionLibrary {nodes [
   Calculate Body Mover Pivot = RigVMCollapseNode(
    Control=(Type=Bone,Name="None")
    ExecuteContext=()
    Ground Transform=(Rotation=(X=0,Y=0,Z=0,W=1),Translation=(X=0,Y=0,Z=0),Scale3D=(X=1,Y=1,Z=1))
    Pivots=()
    Up Axis=(X=0,Y=0,Z=0)
   )
   Compute Box Pivots = RigVMCollapseNode(Dimensions=(X=16,Y=8,Z=10), ExecuteContext=(), Offset=(X=0,Y=0,Z=10))
   Compute Circle Pivots = RigVMCollapseNode(Count=8, ExecuteContext=(), Offset=(X=0,Y=0,Z=0), Radius=10)
   Get Transform Closest To Plane = RigVMCollapseNode(ExecuteContext=(), Plane Normal=(X=0,Y=0,Z=1), Plane Point=(X=0,Y=0,Z=0), Transforms=())
   Multiply Transforms = RigVMCollapseNode(ExecuteContext=(), Local=(), Parent=(Rotation=(X=0,Y=0,Z=0,W=1),Translation=(X=0,Y=0,Z=0),Scale3D=(X=1,Y=1,Z=1)))
   Solve Rocker = RigVMCollapseNode(
    Control=(Type=Bone,Name="None")
    ExecuteContext=()
    Ground Transform=(Rotation=(X=0,Y=0,Z=0,W=1),Translation=(X=0,Y=0,Z=0),Scale3D=(X=1,Y=1,Z=1))
    Pivots=()
    Proxy=(Type=Bone,Name="None")
    Up Axis=(X=0,Y=0,Z=0)
   )
   Store State = RigVMCollapseNode(
    Control=(Type=Bone,Name="None")
    ExecuteContext=()
    Ground Transform=(Rotation=(X=0,Y=0,Z=0,W=1),Translation=(X=0,Y=0,Z=0),Scale3D=(X=1,Y=1,Z=1))
    Proxy=(Type=Bone,Name="None")
    Up Axis=(X=0,Y=0,Z=0)
    World Pivots=()
   )
  ]}
 RigVMFunction_Sequence_1_ContainedGraph {nodes [
   n1 = RigVMFunctionEntryNode()
   RigVMFunction_Sequence = &RigVMFunction_Sequence(ExecuteContext=@n1.ExecuteContext)
   n2 = &RigVMFunction_Sequence(ExecuteContext=@RigVMFunction_Sequence.B)
   RigVMFunctionReturnNode(A=@RigVMFunction_Sequence.A, B=@n2.A, C=@n2.B)
  ]}
 RigVMFunction_Sequence_1_ContainedGraph {nodes [
   n1 = RigVMFunctionEntryNode()
   RigVMFunction_Sequence = &RigVMFunction_Sequence(ExecuteContext=@n1.ExecuteContext)
   n2 = &RigVMFunction_Sequence(ExecuteContext=@RigVMFunction_Sequence.B)
   RigVMFunctionReturnNode(A=@RigVMFunction_Sequence.A, B=@n2.A, C=@n2.B)
  ]}
 RigVMFunction_Sequence_1_ContainedGraph {nodes [
   n1 = RigVMFunctionEntryNode()
   RigVMFunction_Sequence = &RigVMFunction_Sequence(ExecuteContext=@n1.ExecuteContext)
   n2 = &RigVMFunction_Sequence(ExecuteContext=@RigVMFunction_Sequence.B)
   n3 = &RigVMFunction_Sequence(ExecuteContext=@n2.B)
   RigVMFunctionReturnNode(A=@RigVMFunction_Sequence.A, B=@n2.A, C=@n3.A, D=@n3.B)
  ]}
 RigVMFunction_Sequence_1_ContainedGraph {nodes [
   n1 = RigVMFunctionEntryNode()
   RigVMFunction_Sequence = &RigVMFunction_Sequence(ExecuteContext=@n1.ExecuteContext)
   n2 = &RigVMFunction_Sequence(ExecuteContext=@RigVMFunction_Sequence.B)
   n3 = &RigVMFunction_Sequence(ExecuteContext=@n2.B)
   n4 = &RigVMFunction_Sequence(ExecuteContext=@n3.B)
   RigVMFunctionReturnNode(A=@RigVMFunction_Sequence.A, B=@n2.A, C=@n3.A, D=@n4.A, E=@n4.B)
  ]}
 RigVMModel Forward Solve Event Graph {nodes [
   n1 = /Script/ControlRig.RigUnit_PreBeginExecution() {is=struct RigVMUnitNode}
   RigVMFunction_Sequence = &RigVMFunction_Sequence(ExecuteContext=@n1.ExecutePin)
   Multiply Transforms = RigVMFunctionReferenceNode(
    Parent=&RigUnit_GetTransform(bInitial=False, Item=RigVMVariableNode() {variable=Control}.Value).Transform
    ExecuteContext=@RigVMFunction_Sequence.A
    Local=RigVMVariableNode() {variable=Local Pivots}.Value
   )
   > RigVMVariableNode(Value=@"Multiply Transforms".Result) {variable=World Pivots}
   > RigVMVariableNode(Value=&RigUnit_GetTransform(bInitial=False, Item=RigVMVariableNode() {variable=Ground}.Value).Transform) {variable=Ground Transform}
   RigVMFunction_IsHostBeingDebugged = /Script/RigVM.RigVMFunction_IsHostBeingDebugged() {is=struct RigVMUnitNode}
   > /Script/RigVM.RigVMFunction_DebugTransformArrayMutableNoSpace(
    Color=(R=0,G=0,B=1,A=1)
    DebugDrawSettings=(DepthPriority=SDPG_Foreground,Lifetime=-1)
    Mode=Point
    ParentIndices=()
    Scale=2
    Thickness=0
    WorldOffset=(Rotation=(X=0,Y=0,Z=0,W=1),Translation=(X=0,Y=0,Z=0),Scale3D=(X=1,Y=1,Z=1))
    bEnabled=@RigVMFunction_IsHostBeingDebugged.Result
    Transforms=RigVMVariableNode() {variable=World Pivots}.Value
   ) {is=struct RigVMUnitNode}
   n2 = /Script/ControlRig.RigUnit_IsInteracting() {is=struct RigVMUnitNode}
   RigVMFunction_ControlFlowBranch = &RigVMFunction_ControlFlowBranch(Condition=@n2.bIsInteracting, ExecuteContext=@RigVMFunction_Sequence.B)
   Store State = RigVMFunctionReferenceNode(
    ExecuteContext=@RigVMFunction_ControlFlowBranch.False
    World Pivots=RigVMVariableNode() {variable=World Pivots}.Value
    Control=RigVMVariableNode() {variable=Control}.Value
    Ground Transform=RigVMVariableNode() {variable=Ground Transform}.Value
    Up Axis=RigVMVariableNode() {variable=Up Axis Vector}.Value
    Proxy=RigVMVariableNode() {variable=Proxy Control}.Value
   )
   > &RigUnit_SetTransform(
    Weight=1
    bInitial=False
    Value=&RigUnit_GetTransform(bInitial=False, Item=RigVMVariableNode() {variable=Control}.Value).Transform
    Item=RigVMVariableNode() {variable=Proxy Control}.Value
   )
   n3 = &RigVMFunction_ControlFlowBranch(
    Condition=RigVMDispatchNode(Array=@n2.Items, Element=RigVMVariableNode() {variable=Proxy Control}.Value) {template="DISPATCH_RigVMDispatch_ArrayFind(in Array,in Element,out Index,out Success)"}.Success
    ExecuteContext=@RigVMFunction_ControlFlowBranch.True
   )
   n4 = RigVMVariableNode() {variable=World Pivots}
   RigVMFunction_DebugTransformArrayMutableNoSpace = &RigVMFunction_DebugTransformArrayMutableNoSpace(
    Color=(R=1,G=1,B=1,A=1)
    DebugDrawSettings=(DepthPriority=SDPG_Foreground,Lifetime=-1)
    Mode=Point
    ParentIndices=()
    Scale=2
    Thickness=0
    WorldOffset=(Rotation=(X=0,Y=0,Z=0,W=1),Translation=(X=0,Y=0,Z=0),Scale3D=(X=1,Y=1,Z=1))
    bEnabled=True
    ExecutePin=@n3.True
    Transforms=@n4.Value
   )
   n5 = RigVMVariableNode() {variable=Proxy Control}
   Get_Metadata_2 = RigVMDispatchNode(Default=0, Name=Rocker Current Transform Index, NameSpace=Self, Item=@n5.Value) {template="DISPATCH_RigDispatch_GetMetadata(in Item,in Name,in NameSpace,in Default,out Value,out Found)"}
   > RigVMFunction_DebugTransformMutableNoSpace = /Script/RigVM.RigVMFunction_DebugTransformMutableNoSpace(
    Color=(R=0,G=1,B=0,A=1)
    DebugDrawSettings=(DepthPriority=SDPG_Foreground,Lifetime=-1)
    Mode=Point
    Scale=5
    Thickness=0
    WorldOffset=(Rotation=(X=0,Y=0,Z=0,W=1),Translation=(X=0,Y=0,Z=0),Scale3D=(X=1,Y=1,Z=1))
    bEnabled=True
    Transform=RigVMDispatchNode(Index=@Get_Metadata_2.Value, Array=@n4.Value) {template="DISPATCH_RigVMDispatch_ArrayGetAtIndex(in Array,in Index,out Element)"}.Element
   ) {is=struct RigVMUnitNode}
   Get_Metadata = RigVMDispatchNode(
    Default=(Rotation=(X=0,Y=0,Z=0,W=1),Translation=(X=0,Y=0,Z=0),Scale3D=(X=1,Y=1,Z=1))
    Name=Rocker Ground Transform
    NameSpace=Self
    Item=@n5.Value
   ) {template="DISPATCH_RigDispatch_GetMetadata(in Item,in Name,in NameSpace,in Default,out Value,out Found)"}
   Cast = RigVMDispatchNode(Value=RigVMVariableNode() {variable=Up Axis}.Value) {template="DISPATCH_RigVMDispatch_CastEnumToInt(in Value,out Result)"}
   Modulo = /Script/RigVM.RigVMFunction_MathIntMod(B=3, A=@Cast.Result) {is=struct RigVMUnitNode, template="Modulo::Execute(in A,in B,out Result)"}
   > RigVMFunction_DebugRectangleNoSpace = /Script/RigVM.RigVMFunction_DebugRectangleNoSpace(
    Color=(R=1,G=1,B=1,A=0.5)
    DebugDrawSettings=(DepthPriority=SDPG_Foreground,Lifetime=-1)
    Scale=40
    Thickness=0
    WorldOffset=(Rotation=(X=0,Y=0,Z=0,W=1),Translation=(X=0,Y=0,Z=0),Scale3D=(X=1,Y=1,Z=1))
    bEnabled=True
    Transform=&RigVMFunction_MathTransformMakeAbsolute(
     Local=(Rotation=(X=0,Y=0,Z=0,W=1),Translation=(X=0,Y=0,Z=0),Scale3D=(X=1,Y=1,Z=1))
     Parent=@Get_Metadata.Value
     Local.Rotation=RigVMDispatchNode(Values=((X=0,Y=-0.7071,Z=0,W=0.7071),(X=-0.7071,Y=0,Z=0,W=0.7071),()), Index=@Modulo.Result) {template="DISPATCH_RigVMDispatch_SelectInt32(in Index,in Values,out Result)"}.Result
    ) {template="Make Absolute::Execute(in Local,in Parent,out Global)"}.Global
   ) {is=struct RigVMUnitNode}
   > Solve Rocker = RigVMFunctionReferenceNode(
    Pivots=RigVMVariableNode() {variable=World Pivots}.Value
    Control=RigVMVariableNode() {variable=Control}.Value
    Ground Transform=RigVMVariableNode() {variable=Ground Transform}.Value
    Up Axis=RigVMVariableNode() {variable=Up Axis Vector}.Value
    Proxy=RigVMVariableNode() {variable=Proxy Control}.Value
   )
   RigUnit_BeginExecution = &RigUnit_BeginExecution()
   RigUnit_SetTransform = &RigUnit_SetTransform(
    ExecutePin=()
    Item=(Type=Bone,Name="rockerB_02")
    Weight=1
    bInitial=False
    Value=&RigUnit_GetTransform(bInitial=False, Item=RigVMVariableNode() {variable=Control}.Value).Transform
   )
  ]}
 RigVMModel Rig {nodes [/Script/ControlRig.RigUnit_InverseExecution() {is=struct RigVMUnitNode}]}
]
hierarchy [
 Control {type=Connector, loc=<0 0 0>, rot=<0 0 0 1>}
 Root {type=Socket, loc=<0 0 0>, rot=<0 0 0 1>}
 Ground {type=Connector, loc=<0 0 0>, rot=<0 0 0 1>}
]
variables [
 Add Box Pivots {type=bool, default=True}
 Add Circle Pivots {type=bool, default=False}
 Box Dimensions {type=FVector, default=(X=16,Y=8,Z=9.5)}
 Box Offset {type=FVector, default=(X=0,Y=0,Z=9.5)}
 Circle Axis {type=EUserDefinedEnum_CREnum_UpAxis_da2e2c97, default=NewEnumerator2}
 Circle Offset {type=FVector, default=(X=0,Y=0,Z=0)}
 Circle Radius {type=double, default=30}
 Control {type=FRigElementKey, default=(Type=Connector,Name="")}
 Custom Pivots {type="TArray<FTransform>"}
 Ground {type=FRigElementKey, default=(Type=Connector,Name="")}
 Ground Transform {type=FTransform, default=(Rotation=(X=0,Y=0,Z=0,W=1),Translation=(X=0,Y=0,Z=0),Scale3D=(X=1,Y=1,Z=1))}
 Local Pivots {type="TArray<FTransform>"}
 Proxy Control {type=FRigElementKey, default=(Type=Bone,Name="")}
 Proxy Control Shape {
  type=FRigUnit_HierarchyAddControl_ShapeSettings
  default=(bVisible=True,Name="Diamond_Solid",Color=(R=0,G=1,B=0,A=1),Transform=(Rotation=(X=0,Y=0,Z=0,W=1),Translation=(X=0,Y=0,Z=40),Scale3D=(X=1,Y=1,Z=1)))
 }
 Up Axis {type=EUserDefinedEnum_CREnum_UpAxis_da2e2c97, default=NewEnumerator2}
 Up Axis Vector {type=FVector, default=(X=0,Y=0,Z=0)}
 World Pivots {type="TArray<FTransform>"}
]
