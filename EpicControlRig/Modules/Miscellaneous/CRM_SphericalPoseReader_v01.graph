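# Control Rig module that reads a driver's pose with a spherical pose reader and uses the result to weight corrective-style transforms on driven items.
# Construction resolves the Driver, Driven (array) and Optional Parent connectors into variables.
# Forward solve feeds driver, axis, active and falloff regions into the reader, storing its output in OutputParam.
# It then modifies each driven item by the index-matched DrivenTransform, weighted by OutputParam.
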ControlRigBlueprint /EpicControlRig/Modules/Miscellaneous/CRM_SphericalPoseReader_v01
graphs [
 ConstructionGraph {nodes [
   RigUnit_PrepareForExecution = &RigUnit_PrepareForExecution()
   RigVMFunction_Sequence = &RigVMFunction_Sequence(ExecuteContext=@RigUnit_PrepareForExecution.ExecutePin)
   RigVMVariableNode(
    Value=/Script/ControlRig.RigUnit_ResolveConnector(Connector=(Type=Connector,Name="Driver"), SkipSocket=False) {is=struct RigVMUnitNode}.Result
    ExecuteContext=@RigVMFunction_Sequence.A
   ) {variable=DriverItem}
   > RigVMVariableNode(Value=/Script/ControlRig.RigUnit_ResolveArrayConnector(Connector=(Type=Connector,Name="Driven"), SkipSocket=False) {is=struct RigVMUnitNode}.Result) {variable=DrivenItems}
   > RigVMVariableNode(Value=&RigUnit_ResolveConnector(Connector=(Type=Connector,Name="Optional Parent"), SkipSocket=False).Result) {variable=OptionalParentItem}
  ]}
 RigVMFunctionLibrary {links [], nodes []}
 RigVMModel {nodes [
   RigUnit_BeginExecution = &RigUnit_BeginExecution()
   RigVMFunction_Sequence = &RigVMFunction_Sequence(ExecuteContext=@RigUnit_BeginExecution.ExecutePin)
   n1 = &RigUnit_SphericalPoseReader(
    Debug=(bDrawDebug=true,bDraw2D=false,bDrawLocalAxes=false,DebugScale=25,DebugSegments=20,DebugThickness=0.25)
    FlipHeightScaling=False
    FlipWidthScaling=False
    OptionalParentItem=RigVMVariableNode() {variable=OptionalParentItem}.Value
    Debug.bDrawDebug=RigVMVariableNode() {variable=DrawDebug}.Value
    DriverItem=RigVMVariableNode() {variable=DriverItem}.Value
    DriverAxis=RigVMVariableNode() {variable=DriverAxis}.Value
    ActiveRegionSize=RigVMVariableNode() {variable=ActiveRegionSize}.Value
    RotationOffset=RigVMVariableNode() {variable=RotationOffset}.Value
    ActiveRegionScaleFactors=RigVMVariableNode() {variable=ActiveRegionScaleFactors}.Value
    FalloffSize=RigVMVariableNode() {variable=FalloffSize}.Value
    FalloffRegionScaleFactors=RigVMVariableNode() {variable=FalloffRegionScaleFactors}.Value
    ExecutePin=@RigVMFunction_Sequence.A
   )
   > RigVMVariableNode(Value=@n1.OutputParam) {variable=OutputParam}
   For_Each = RigVMDispatchNode(Array=RigVMVariableNode() {variable=DrivenItems}.Value, ExecuteContext=@RigVMFunction_Sequence.B) {template="DISPATCH_RigVMDispatch_ArrayIterator(in Array,out Element,out Index,out Count,out Ratio)"}
   > &RigUnit_ModifyTransforms(
    ItemToModify=((Item=(Type=Bone,Name="None"),Transform=(Rotation=(X=0,Y=0,Z=0,W=1),Translation=(X=0,Y=0,Z=0),Scale3D=(X=1,Y=1,Z=1))))
    WeightMinimum=0
    ItemToModify.0.Transform=RigVMDispatchNode(Index=@For_Each.Index, Array=RigVMVariableNode() {variable=DrivenTransform}.Value) {template="DISPATCH_RigVMDispatch_ArrayGetAtIndex(in Array,in Index,out Element)"}.Element
    ItemToModify.0.Item=@For_Each.Element
    Weight=RigVMVariableNode() {variable=OutputParam}.Value
   )
  ]}
]
hierarchy [
 Driver {type=Connector, loc=<0 0 0>, rot=<0 0 0 1>}
 Root {type=Socket, loc=<0 0 0>, rot=<0 0 0 1>}
 Driven {type=Connector, loc=<0 0 0>, rot=<0 0 0 1>}
 Optional Parent {type=Connector, loc=<0 0 0>, rot=<0 0 0 1>}
]
variables [
 ActiveRegionScaleFactors {type=FRegionScaleFactors, default=(PositiveWidth=1,NegativeWidth=1,PositiveHeight=1,NegativeHeight=1)}
 ActiveRegionSize {type=float, default=0.1}
 DrawDebug {type=bool, default=True}
 DrivenItem {type=FRigElementKey, default=(Type=Connector,Name="")}
 DrivenItems {type="TArray<FRigElementKey>"}
 DrivenTransform {type="TArray<FTransform>"}
 DriverAxis {type=FVector, default=(X=1,Y=0,Z=0)}
 DriverItem {type=FRigElementKey, default=(Type=Bone,Name="")}
 FalloffRegionScaleFactors {type=FRegionScaleFactors, default=(PositiveWidth=1,NegativeWidth=1,PositiveHeight=1,NegativeHeight=1)}
 FalloffSize {type=float, default=0.2}
 OptionalParentItem {type=FRigElementKey, default=(Type=Bone,Name="")}
 OutputParam {type=float, default=0}
 RotationOffset {type=FVector, default=(X=0,Y=0,Z=0)}
]
